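# Material function in the SpacePlunge character FX folder that computes interior-cubemap UVW for fake parallax rooms behind a surface.
# It tiles and fracs UVs into per-room box positions, then intersects the tangent-space view ray, depth-scaled by RoomZscale, with that box.
# As wired, the nearest-hit minimum tests only R and B, and UVW repeats the hit's B component on all three axes.
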MaterialFunction /BRCosmetics/Characters/Player/Male/Medium/Bodies/M_MED_SpacePlunge/FX/Materials/MF_InteriorCubemapUV_NS
graph {
 properties {
  description=Interior Cubemap conforms a cubemap to be in the form of a box shape inset from the surface UVs. It is useful for giving parallax to rooms visible through windows on city buildings.
  libraryCategoriesText [Cubemaps]
 }
 nodes [
  n1 = FunctionInput(Preview=CameraVectorWS()) {bUsePreviewValueAsDefault=true, inputName=View}
  n2 = Constant() {r=1}
  n3 = Constant() {r=1}
  n4 = Divide(B=FunctionInput(Preview=Constant() {r=1}) {bUsePreviewValueAsDefault=true, inputName=RoomZscale, inputType=FunctionInput_Scalar})
  n5 = Multiply(
   A=Constant3Vector() {constant=<rgba -1 1 1 0>}
   B=Multiply(
    A=Transform(None=@n1) {transformSourceType=TRANSFORMSOURCE_World, transformType=TRANSFORM_Tangent}
    B=MaterialFunctionCall(X=@n2, Y=@n3, Z=@n4) {materialFunction=&MakeFloat3}.Result
   )
  )
  n6 = Divide(B=@n5) {desc=id}
  n7 = TextureCoordinate()
  n8 = FunctionInput() {bUsePreviewValueAsDefault=true, inputName=Tiling, inputType=FunctionInput_Vector2, sortPriority=1, previewValue=<4 4 0 1>}
  n9 = Frac(None=Multiply(A=FunctionInput(Preview=@n7) {bUsePreviewValueAsDefault=true, inputName=UVs, inputType=FunctionInput_Vector2}, B=@n8)) {desc=f}
  n10 = Constant2Vector() {g=-2, r=2}
  n11 = AppendVector(A=Subtract(A=Multiply(A=@n9, B=@n10), B=Constant2Vector() {g=-1, r=1}), B=Constant() {r=-1})
  n12 = MaterialFunctionCall(Float3=Subtract(A=Abs(None=@n6), B=Multiply(A=@n6, B=@n11)) {desc=k}) {materialFunction=&BreakOutFloat3Components}
  n13 = Min(A=Min(A=@n12.R, B=@n12.R), B=@n12.B)
  n14 = MaterialFunctionCall(Float3=Add(A=Multiply(A=@n13, B=@n5), B=@n11) {desc="pos.xyz += kMin * In.dir;"}) {materialFunction=&BreakOutFloat3Components}
  FunctionOutput(None=MaterialFunctionCall(X=@n14.B, Y=@n14.B, Z=@n14.B) {materialFunction=&MakeFloat3}.Result) {outputName=UVW}
 ]
}
