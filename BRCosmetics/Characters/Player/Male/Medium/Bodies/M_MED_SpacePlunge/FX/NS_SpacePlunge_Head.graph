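# Niagara cosmetic effect for the head of the SpacePlunge character skin, attached through skeletal mesh sampling.
# Glows adds pulsing variant-colored sprites and lights; Hunt_Lightning, Hunt_MeshFX and CubicFlame add lightning meshes, head meshes and flame sprites for variants 2 and 4.
# A shared system Timeline drives spawn-rate, brightness and mesh-stretch surges.
# Head-hidden state zeroes size or color, visor state also hides Glows, and front-end modifiers adjust intensity and placement.
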NiagaraSystem /BRCosmetics/Characters/Player/Male/Medium/Bodies/M_MED_SpacePlunge/FX/NS_SpacePlunge_Head
dependencies {
 usedDataInterfaces [
  &NiagaraDataInterfaceArrayInt32
  &NiagaraDataInterfaceCurve
  &NiagaraDataInterfaceArrayFloat3
  &NiagaraDataInterfaceArrayFloat
  &NiagaraDataInterfaceArrayColor
  &NiagaraDataInterfaceVectorCurve
  &NiagaraDataInterfaceVector2DCurve
  &NiagaraDataInterfaceSkeletalMesh
  &NiagaraDataInterfaceColorCurve
 ]
 usedDynamicInputs [
  &SetBoolByFloatComparison
  &MakeIntFromFloat
  &MakeCustomLinearColorFromBool
  &Boolean_AndOperation
  &SetBoolByIntComparison
 ]
 usedModules [
  &self:SystemSpawnScript.SystemScriptSource.SystemScriptGraph.NiagaraNodeAssignment_0.SetVariables_#1
  &self:SystemSpawnScript.SystemScriptSource.SystemScriptGraph.NiagaraNodeAssignment_2.SetVariables_#2
  &self:SystemSpawnScript.SystemScriptSource.SystemScriptGraph.NiagaraNodeAssignment_3.SetVariables_#3
  &self:SystemSpawnScript.SystemScriptSource.SystemScriptGraph.NiagaraNodeAssignment_4.SetVariables_#4
  &self:SystemSpawnScript.SystemScriptSource.SystemScriptGraph.NiagaraNodeAssignment_6.SetVariables_#5
  &SystemState
  &BlendSpaceHelper
  &self:SystemSpawnScript.SystemScriptSource.SystemScriptGraph.NiagaraNodeAssignment_1.SetVariables_#6
  &self:SystemSpawnScript.SystemScriptSource.SystemScriptGraph.NiagaraNodeAssignment_5.SetVariables_#7
  &self:SystemSpawnScript.SystemScriptSource.SystemScriptGraph.NiagaraNodeAssignment_7.SetVariables_#8
  /Game/Effects/Niagara/Modules/Logic/UpdateValueStateChange
  &Timeline
 ]
}
emitters {
 CubicFlame {
  emitterUpdateScript [
   &EmitterState()
   &SpawnRate(
    Spawn Probability=&SelectFloatFromArray(Direct Array Index=$System.VariantIndex, Float Selection Array=DataInterface() {propertyValues {FloatData [0,0,1,0,1]}})
    SpawnRate=&FloatFromCurve(
     FloatCurve=DataInterface() {propertyValues {Curve {keys [
         {interpMode=RCIM_Cubic, tangentMode=RCTM_User, time=0, value=0.25}
         {interpMode=RCIM_Cubic, tangentMode=RCTM_User, time=0.05, value=1}
         {interpMode=RCIM_Cubic, tangentMode=RCTM_User, time=0.75, value=1}
         {arriveTangent=-5.7829, tangentMode=RCTM_User, time=1, value=0.25}
        ]}}}
     CurveIndex=$System.Timeline.NormalizedTime
     Scale Curve=150
    )
    Use Spawn Probability=true
   )
  ]
  particleSpawnScript [
   &PartitionParticles()
   &InitializeParticle(
    Color Mode=::NewEnumerator0
    Lifetime=&RandomRangeFloat(Minimum=0.5, Maximum=1.5)
    Sprite Size Max=<10 20>
    Sprite Size Min=<10 10>
    Sprite Size Mode=::NewEnumerator2
    UsePositionOffset=false
    Write Material Random=true
   )
   &DynamicMaterialParameters(
    Index 0 Param 1=&RandomRangeFloat(Minimum=0, Maximum=1)
    Index 0 Param 2=&RandomRangeFloat(Minimum=0, Maximum=1)
    Module.Index0Float0Write=true
    Module.Index0Float1Write=true
    Module.Index0Float2Write=true
    Module.Index0Float3Write=true
   )
  ]
  particleUpdateScript [
   &ParticleState()
   &ScaleSpriteSize(
    Non-Uniform Curve Scale=&MakeCustomVector2DFromBool(Bool=$User.bisHeadHidden, True Vector2D=<0 0>, False Vector2D=<1 1>)
    Non-Uniform Curve Sprite Scale=DataInterface() {propertyValues {
      XCurve {keys [{interpMode=RCIM_Cubic, tangentMode=RCTM_User, time=0, value=1}, {arriveTangent=-2.4882, tangentMode=RCTM_User, time=1, value=0}]}
      YCurve {keys [
        {arriveTangent=3.3898, interpMode=RCIM_Cubic, leaveTangent=3.3898, tangentMode=RCTM_User, time=0, value=0.1}
        {arriveTangent=0.0355, interpMode=RCIM_Cubic, leaveTangent=0.0355, tangentMode=RCTM_User, time=1, value=2}
       ]}
     }}
    Scale Sprite Size Mode=::NewEnumerator3
   )
   &CameraOffset(Camera Offset Amount=-0.5)
   &SkeletalMeshLocation(
    Bone / Socket Index=$Particles.PartitionParticles.Partition
    Bone Sampling Mode=::NewEnumerator4
    MaskSpawnGroup=false
    Orientation Sampling=::NewEnumerator1
    Sampled Position Offset=&MaskVector(Mask Source=1, Mask Parameter=$Particles.PartitionParticles.Partition, Mask=0, Masked Vector=<5 0 0>, Passthrough Vector=<-5 0 0>)
    Skeletal Mesh=$System.Skeletal Mesh
   )
   &ShapeLocation(
    Disc Coverage=1
    MaskSpawnGroup=false
    Override Seed=false
    Ring Radius=&MaskFloat(Mask Source=1, Mask Parameter=$Particles.PartitionParticles.Partition, Mask=0, Masked Float=7, Passthrough Float=10)
    Rotation Angle=0.25
    Rotation Axis=<0 0 1>
    Rotation Mode=::NewEnumerator1
    Shape Origin=$Particles.Position
    Shape Primitive=::NewEnumerator4
    UseOwnerScale=true
   )
   &SolveForcesAndVelocity(Clamp Velocity=false, Limit Acceleration=false)
   &ScaleColor(
    Scale Alpha=&FloatFromCurve(
     FloatCurve=DataInterface() {propertyValues {Curve {keys [<at 0 0>,<at 0.1 1>,<at 1 0>]}}}
     Scale Curve=&MakeCustomFloatFromBool(Bool=$User.bisHeadHidden, True Float=0, False Float=1)
    )
    Scale RGB=&Multiply_VectorByFloat(
     Vector=<1 1 1>
     Float=&FloatFromCurve(
      FloatCurve=DataInterface() {propertyValues {Curve {keys [<at 0 1>,<at 0.084 3>,<at 1 0>]}}}
      Scale Curve=&MakeCustomFloatFromBool(Bool=$User.bisHeadHidden, True Float=0, False Float=1)
     )
    )
    ScaleA=true
    ScaleRGB=true
   )
  ]
  rendererClasses [&NiagaraSpriteRendererProperties]
  renderers [{rendererClass=&NiagaraSpriteRendererProperties}]
 }
 Glows {
  emitterUpdateScript [
   &EmitterState(
    Inactive Response=::NewEnumerator1
    Life Cycle Mode=::NewEnumerator1
    Loop Behavior=::NewEnumerator1
    Loop Duration Mode=::NewEnumerator1
    UseLoopDelay=true
   )
   &SpawnBurst_Instantaneous(Use Spawn Probability=true, UseLoopCountLimit=true)
  ]
  particleSpawnScript [
   &InitializeParticle(
    Lifetime=6
    Mesh Scale Mode=::NewEnumerator3
    Mesh Uniform Scale=&MaskFloat(Mask Source=1, Mask Parameter=$System.VariantIndex, Mask=2, Masked Float=1.06, Passthrough Float=1.02)
    Sprite Size=<25 25>
    Sprite Size Mode=::NewEnumerator4
    UsePositionOffset=false
    Write Material Random=true
   )
   &Color(Color=$System.VariantColorLinear)
  ]
  particleUpdateScript [
   &ParticleState(Kill Particles When Lifetime Has Elapsed=false, Let Infinitely Lived Particles Die When Emitter Deactivates=true)
   /SetVariables(
    Particles.Glow_Base_Color:LinearColor=&Multiply_LinearColorByFloat(Linear Color=$System.VariantColorLinear, Float=0.6)
    Particles.Glow_Base_Size:Vector2f=&MakeCustomVector2DFromBool(
     Bool=&Boolean_OrOperation(Bool1=$User.bisHeadHidden, Bool2=$System.bVisorHidden)
     True Vector2D=<0 0>
     False Vector2D=<80 80>
    )
    Particles.VisibilityTag=&SelectIntFromArray(Direct Array Index=$System.VariantIndex, Int Selection Array=DataInterface() {propertyValues {IntData [0,0,1,0,1]}})
   ) {isSetParameters=true}
   &ScaleSpriteSize(
    Scale Factor=&Vector2DFromFloat(Value=&FloatFromCurve(
      FloatCurve=DataInterface() {propertyValues {Curve {keys [
          {interpMode=RCIM_Cubic, tangentMode=RCTM_User, time=0, value=1.2}
          {interpMode=RCIM_Cubic, tangentMode=RCTM_User, time=0.5, value=1.7}
          <at 1 1.2>
         ]}}}
      Scale Curve=&MakeCustomFloatFromBool(Bool=&Boolean_OrOperation(Bool1=$User.bisHeadHidden, Bool2=$System.bVisorHidden), True Float=0, False Float=1)
     ))
    Scale Sprite Size Mode=::NewEnumerator2
   )
   &ScaleColor(
    Scale Alpha=&FloatFromCurve(
     FloatCurve=DataInterface() {propertyValues {Curve {keys [<at 0 0.5 RCIM_Cubic>,<at 0.2321 0.9177 RCIM_Cubic>,<at 1 0.5>]}}}
     Scale Curve=&MakeCustomFloatFromBool(Bool=&Boolean_OrOperation(Bool1=$User.bisHeadHidden, Bool2=$System.bVisorHidden), True Float=0, False Float=1)
    )
    Scale RGB=&VectorFromFloat(Value=&FloatFromCurve(
      FloatCurve=DataInterface() {propertyValues {Curve {keys [
          {interpMode=RCIM_Cubic, tangentMode=RCTM_User, time=0, value=0.75}
          <at 0.5 1 RCIM_Cubic>
          {arriveTangent=-0.1284, interpMode=RCIM_Cubic, leaveTangent=-0.1284, tangentMode=RCTM_User, time=1, value=0.75}
         ]}}}
      Scale Curve=&MakeCustomFloatFromBool(
       Bool=&Boolean_OrOperation(Bool1=$User.bisHeadHidden, Bool2=$System.bVisorHidden)
       True Float=0
       False Float=&FloatFromCurve(
        FloatCurve=DataInterface() {propertyValues {Curve {keys [
            <at 0 1>
            {interpMode=RCIM_Cubic, tangentMode=RCTM_User, time=0.0368, value=1}
            <at 0.0666 3>
            {interpMode=RCIM_Cubic, tangentMode=RCTM_User, time=0.75, value=3}
            {arriveTangent=-16.1427, tangentMode=RCTM_User, time=1, value=1}
           ]}}}
        CurveIndex=$System.Timeline.NormalizedTime
       )
      )
     ))
    ScaleA=true
    ScaleRGB=true
   )
   &SkeletalMeshLocation(
    Bone / Socket Index=&SelectIntFromArray(Direct Array Index=$System.VariantIndex, Int Selection Array=DataInterface() {propertyValues {IntData [2,2,0,2,0]}})
    Bone Sampling Mode=::NewEnumerator4
    MaskSpawnGroup=false
    Sampled Position Coordinate Space=::Local
    Sampled Position Offset=&SelectVectorFromArray(
     Direct Array Index=$System.VariantIndex
     Vector Selection Array=DataInterface() {propertyValues {FloatData [<10 4 0>, <10 4 0>, <8 -1 0>, <10 4 0>, <5 -1 0>]}}
    )
    Skeletal Mesh=$System.Skeletal Mesh
   )
   &CameraOffset(Camera Offset Amount=-2)
   &Light_Attributes(
    Color=1
    Color Value To Scale:LinearColor=$System.VariantColorLinear
    Diffuse Scale=1
    Enabled=1
    Falloff Exponent=1
    Light Enabled:NiagaraBool=&Boolean_NotOperation(Bool=&Boolean_OrOperation(Bool1=$User.bisHeadHidden, Bool2=$System.bVisorHidden))
    Light Exponent=1.5
    Light Position Offset:Vector3f=&SelectVectorFromArray(
     Direct Array Index=$System.VariantIndex
     Vector Selection Array=DataInterface() {propertyValues {FloatData [<0 13 0>, <0 13 0>, <0 13 10>, <0 13 0>, <0 13 10>]}}
    )
    Light Radius=40
    Light Specular Scale=0
    Position=1
    Radius=1
    ScaleA=true
    ScaleRGB=true
    Specular Scale=1
    UsePositionOffset=true
   )
   &ScaleColor(
    Scale Mode=1
    Scale RGBA=/Niagara/DynamicInputs/Multiply/Multiply_Vector4_ByFloat(
     Vector4:Vector4f=<1 1 1 1>
     Float:NiagaraFloat=&FN_FrontEndFloatModifier(Gameplay=1, Front End=0.5)
    )
    ScaleRGBA=true
   ) {module=ScaleColor001}
  ]
  rendererClasses [&NiagaraSpriteRendererProperties, &NiagaraLightRendererProperties]
  renderers [
   {rendererClass=&NiagaraSpriteRendererProperties}
   {rendererClass=&NiagaraLightRendererProperties, rendererIndex=1}
   {rendererClass=&NiagaraSpriteRendererProperties, rendererIndex=2}
   {rendererClass=&NiagaraLightRendererProperties, rendererIndex=3}
  ]
 }
 Hunt_Lightning {
  emitterUpdateScript [
   &EmitterState()
   &SpawnRate(
    Spawn Probability=/Niagara/DynamicInputs/Helpers/DistributionCurveFloat(
     Minimum:NiagaraFloat=0.25
     Maximum:NiagaraFloat=1
     Distribution Curve:NiagaraDataInterfaceCurve=DataInterface() {propertyValues {Curve {
        defaultValue=0
        keys [{interpMode=RCIM_Cubic, tangentMode=RCTM_User, time=0.5, value=0}, {arriveTangent=-0.0819, tangentMode=RCTM_User, time=1, value=1}]
       }}}
     Randomness Mode:ENiagaraRandomnessMode=::NewEnumerator0
     Evaluation Type:ENiagaraRandomnessEvaluation=::NewEnumerator1
     Seed:NiagaraInt32=0
     Fixed Random Seed:NiagaraBool=false
    )
    SpawnRate=&MaskFloat(
     Mask Source=1
     Mask Parameter=$System.VariantIndex
     Mask=2
     Masked Float=&FloatFromCurve(
      FloatCurve=DataInterface() {propertyValues {Curve {keys [
          {interpMode=RCIM_Cubic, tangentMode=RCTM_User, time=0, value=0}
          {interpMode=RCIM_Cubic, tangentMode=RCTM_User, time=0.05, value=1}
          {interpMode=RCIM_Cubic, tangentMode=RCTM_User, time=0.75, value=1}
          {arriveTangent=-5.7829, tangentMode=RCTM_User, time=1, value=0}
         ]}}}
      CurveIndex=$System.Timeline.NormalizedTime
      Scale Curve=100
     )
     Passthrough Float=&MaskFloat(
      Mask Source=1
      Mask Parameter=$System.VariantIndex
      Mask=4
      Masked Float=&FloatFromCurve(
       FloatCurve=DataInterface() {propertyValues {Curve {keys [
           {interpMode=RCIM_Cubic, tangentMode=RCTM_User, time=0, value=0}
           {interpMode=RCIM_Cubic, tangentMode=RCTM_User, time=0.05, value=1}
           {interpMode=RCIM_Cubic, tangentMode=RCTM_User, time=0.75, value=1}
           {arriveTangent=-5.7829, tangentMode=RCTM_User, time=1, value=0}
          ]}}}
       CurveIndex=$System.Timeline.NormalizedTime
       Scale Curve=100
      )
      Passthrough Float=0
     )
    )
    Use Spawn Probability=true
   )
  ]
  particleSpawnScript [
   &PartitionParticles()
   &InitializeParticle(
    Color=$System.VariantColorLinear
    Lifetime Max=0.5
    Lifetime Min=0.25
    Lifetime Mode=::NewEnumerator1
    Mesh Scale=&Multiply_VectorByFloat(
     Vector=&Lerp_Vector(
      A=&RandomRangeVector(Minimum=<0.5 0.4 0.4>, Maximum=<0.75 0.6 0.6>)
      B=&RandomRangeVector(Minimum=<0.8 1 1>, Maximum=<1.25 2 2>)
      Alpha=&FloatFromCurve(
       FloatCurve=DataInterface() {propertyValues {Curve {keys [{interpMode=RCIM_Cubic, tangentMode=RCTM_User, time=0.5, value=0}, <at 1 1>]}}}
       CurveIndex=&RandomRangeFloat(Minimum=0, Maximum=1)
      )
     )
     Float=&SelectFloatFromArray(Direct Array Index=$Particles.PartitionParticles.Partition, Float Selection Array=DataInterface() {propertyValues {FloatData [0.2,0.35]}})
    )
    Mesh Scale Mode=::NewEnumerator4
    UsePositionOffset=true
    Write Material Random=true
   )
   &CameraOffset(Camera Offset Amount=20) {enabled=false}
   &InitialMeshOrientation(
    Mesh Orientation Mode=::NewEnumerator2
    Orientation Coordinate Space=::Local
    Orientation Vector=&SelectVectorFromArray(
     Direct Array Index=$Particles.PartitionParticles.Partition
     Vector Selection Array=DataInterface() {propertyValues {FloatData [<0 0.5 -1>, <0 0.5 -1>]}}
    )
    Rotation=&MaskVector(
     Mask Source=1
     Mask Parameter=$Particles.PartitionParticles.Partition
     Mask=1
     Masked Vector=&RandomRangeVector(Minimum=<-0.25 -0.25 -0.25>, Maximum=<0.25 0.25 0.25>)
     Passthrough Vector=&RandomRangeVector(Minimum=<-0.1 -0.1 -0.1>, Maximum=<0.1 0.1 0.1>)
    )
    Use Rotation Vector=true
   ) {module=InitialMeshOrientation002}
   /SetVariables(Particles.Random=&RandomRangeFloat(Minimum=0, Maximum=1)) {isSetParameters=true}
  ]
  particleUpdateScript [
   &SkeletalMeshLocation(
    Bone / Socket Index=&SelectIntFromArray(Direct Array Index=$Particles.PartitionParticles.Partition, Int Selection Array=DataInterface() {propertyValues {IntData [0,1]}})
    Bone Sampling Mode=::NewEnumerator4
    MaskSpawnGroup=false
    Orientation Sampling=::NewEnumerator1
    Sampled Position Offset=&SelectVectorFromArray(
     Direct Array Index=$Particles.PartitionParticles.Partition
     Vector Selection Array=DataInterface() {propertyValues {FloatData [<6 -2 0>, <-5 0 0>]}}
    )
    Skeletal Mesh=$System.Skeletal Mesh
   )
   &ShapeLocation(
    MaskSpawnGroup=false
    Override Seed=false
    Rotation Mode=::NewEnumerator1
    Shape Origin=$Particles.Position
    Sphere Radius=5
    UseOwnerScale=true
   )
   &ParticleState()
   &ScaleMeshSize(Scale Factor=&MakeCustomVectorFromBool(Bool=$User.bisHeadHidden, True Vector=<0 0 0>, False Vector=<1 1 1>))
   &ScaleColor(
    Scale Alpha=&Multiply_Float(
     A=10
     B=&FloatFromCurve(FloatCurve=DataInterface() {propertyValues {Curve {keys [
          {interpMode=RCIM_Cubic, tangentMode=RCTM_User, time=0, value=0}
          {interpMode=RCIM_Cubic, tangentMode=RCTM_User, time=0.1, value=0}
          {interpMode=RCIM_Cubic, tangentMode=RCTM_User, time=0.2, value=0.8}
          {interpMode=RCIM_Cubic, tangentMode=RCTM_User, time=1, value=0}
         ]}}})
    )
    Scale RGB=&Multiply_VectorByFloat(
     Vector=&Multiply_VectorByFloat(
      Vector=&Multiply_VectorByFloat(Vector=<1 1 1>, Float=&FN_FrontEndFloatModifier(Gameplay=1, Front End=0.5))
      Float=&Multiply_Float(
       A=&RandomRangeFloat(Minimum=0.5, Maximum=1, Evaluation Type=::NewEnumerator1)
       B=&Multiply_Float(A=&FNTimeOfDayScalar(Night Scale=0.2, Morning Scale=1, Evening Scale=1), B=10)
      )
     )
     Float=&MakeCustomFloatFromBool(Bool=$User.bisHeadHidden, True Float=0, False Float=1)
    )
    ScaleA=true
    ScaleRGB=true
   )
   &DynamicMaterialParameters(
    Index 0 Param 1=&FloatFromCurve(FloatCurve=DataInterface() {propertyValues {Curve {keys [{tangentMode=RCTM_User, time=0, value=1}, {tangentMode=RCTM_User, time=1, value=1}]}}})
    Index 0 Param 2=0.7
    Index 0 Param 3=&FloatFromCurve(FloatCurve=DataInterface() {propertyValues {Curve {keys [
         {arriveTangent=4.0072, interpMode=RCIM_Cubic, leaveTangent=4.0072, tangentMode=RCTM_User, time=0, value=0}
         {arriveTangent=2.7083, interpMode=RCIM_Cubic, leaveTangent=2.7083, time=0.2543, value=0.6225}
         {arriveTangent=1.8473, interpMode=RCIM_Cubic, leaveTangent=1.8473, time=0.6, value=1.625}
         {arriveTangent=1.6993, interpMode=RCIM_Cubic, leaveTangent=1.6993, tangentMode=RCTM_User, time=1, value=2}
        ]}}})
    Index 0 Param 4=5
    Index 1 Param 1=&MaskFloat(Mask Source=1, Mask Parameter=$System.VariantIndex, Mask=2, Masked Float=0.4, Passthrough Float=0.15)
    Index 1 Param 2=$Particles.Random
    Index 1 Param 3=1.3
    Index 1 Param 4=1
    Index 2 Param 1=&FloatFromCurve(FloatCurve=DataInterface() {propertyValues {Curve {keys [<at 0 1>,<at 1 0.3>]}}})
    Module.Index0Float0Write=true
    Module.Index0Float1Write=true
    Module.Index0Float2Write=true
    Module.Index0Float3Write=true
    Module.Index1Float0Write=true
    Module.Index1Float1Write=true
    Module.Index1Float2Write=true
    Module.Index1Float3Write=true
    Module.Index2Float0Write=true
    Module.Index2Float1Write=true
    Module.Index2Float2Write=true
    Module.Index2Float3Write=true
    Write Parameter Index 1=::NewEnumerator1
    Write Parameter Index 2=::NewEnumerator1
   )
   &SolveForcesAndVelocity(Clamp Velocity=false, Limit Acceleration=false)
   /SetVariables(Particles.Lightning glow color:LinearColor=&MakeLinearColorFromVectorAndFloat(
     "Vector (RGB)"=&MakeVectorFromLinearColorRGB(Color=&MultiplyLinearColor(
       Linear Color=&SelectLinearColorFromArray(
        Array Sampling Mode=::NewEnumerator1
        Direct Array Index=$System.VariantIndex
        Color Selection Array=DataInterface() {propertyValues {ColorData [<rgba 4 1.6071 0.1968 1>, <rgba 0.3366 2.0701 4 1>, <rgba 4 1.6071 0.1968 1>]}}
        Random Seed=?
       )
       Scale Factor=&MakeVector4_FromFloat(Float=&RandomRangeFloat(Minimum=0.5, Maximum=1, Evaluation Type=::NewEnumerator1))
      ))
     "Float (Alpha)"=&Multiply_Float(
      A=&FloatFromCurve(FloatCurve=DataInterface() {propertyValues {Curve {keys [<at 0 1 RCIM_Cubic>, {interpMode=RCIM_Cubic, leaveTangent=-3.6307, time=0.9, value=0}]}}})
      B=1
     )
    )) {isSetParameters=true}
   /SetVariables(Particles.Glow Dy02:Vector4f=&MakeVector4(
     X=&FloatFromCurve(FloatCurve=DataInterface() {propertyValues {Curve {keys [{tangentMode=RCTM_User, time=0, value=0.2}, {tangentMode=RCTM_User, time=0.5, value=1}]}}})
     Y=0.7
     Z=0
     W=5
    )) {isSetParameters=true}
   /SetVariables(Particles.Glow Dy03:Vector4f=&MakeVector4(
     X=&FloatFromCurve(
      FloatCurve=DataInterface() {propertyValues {Curve {keys [<at 0 1>,<at 1 0.5>]}}}
      Scale Curve=&MaskFloat(Mask Source=1, Mask Parameter=$System.VariantIndex, Mask=2, Masked Float=0.7, Passthrough Float=0.5)
     )
     Y=$Particles.Random
     Z=1.3
     W=1
    )) {isSetParameters=true}
  ]
  rendererClasses [&NiagaraMeshRendererProperties]
  renderers [
   {rendererClass=&NiagaraMeshRendererProperties, rendererIndex=0}
   {rendererClass=&NiagaraMeshRendererProperties, rendererIndex=1}
  ]
 }
 Hunt_MeshFX {
  emitterUpdateScript [
   &EmitterState(
    Inactive Response=::NewEnumerator1
    Life Cycle Mode=::NewEnumerator1
    Loop Behavior=::NewEnumerator1
    Loop Duration Mode=::NewEnumerator1
    UseLoopDelay=false
   )
   &SpawnBurst_Instantaneous(
    Spawn Probability=&SelectFloatFromArray(Direct Array Index=$System.VariantIndex, Float Selection Array=DataInterface() {propertyValues {FloatData [0,0,1,0,1]}})
    Use Spawn Probability=true
    UseLoopCountLimit=false
   )
  ]
  particleSpawnScript [
   &PartitionParticles(Number Of Partitions=1)
   &InitializeParticle(
    Color=<rgba 1.5 1.5 1.5 1>
    Mesh Scale=&SelectVectorFromArray(
     Direct Array Index=$System.VariantIndex
     Vector Selection Array=DataInterface() {propertyValues {FloatData [<0 0 0>, <0 0 0>, <1 1 1>, <0 0 0>, <1.4 0.95 0.95>]}}
    )
    Mesh Scale Mode=::NewEnumerator4
    Sprite Size Mode=::NewEnumerator3
    Uniform Sprite Size=100
    UsePositionOffset=false
    Write Material Random=true
   )
  ]
  particleUpdateScript [
   &ParticleState(Kill Particles When Lifetime Has Elapsed=false, Let Infinitely Lived Particles Die When Emitter Deactivates=true)
   /SetVariables(Particles.VisibilityTag=&SelectIntFromArray(Direct Array Index=$System.VariantIndex, Int Selection Array=DataInterface() {propertyValues {IntData [0,0,1,0,2]}})) {isSetParameters=true}
   &SkeletalMeshLocation(
    Bone / Socket Index=1
    Bone Sampling Mode=::NewEnumerator4
    MaskSpawnGroup=false
    Sampled Position Offset=&FN_FrontEndVector3Modifier(Gameplay=<0 0 0>, Front End=<1 0 0>)
    Skeletal Mesh=$System.Skeletal Mesh
   )
   &ScaleColor(
    Scale Alpha=&MakeCustomFloatFromBool(Bool=$User.bisHeadHidden, True Float=0, False Float=1)
    Scale RGB=&VectorFromFloat(Value=&FloatFromCurve(
      FloatCurve=DataInterface() {propertyValues {Curve {keys [
          <at 0 1>
          {interpMode=RCIM_Cubic, tangentMode=RCTM_User, time=0.0368, value=1}
          <at 0.0666 3>
          {interpMode=RCIM_Cubic, tangentMode=RCTM_User, time=0.75, value=3}
          {arriveTangent=-16.1427, tangentMode=RCTM_User, time=1, value=1}
         ]}}}
      CurveIndex=$System.Timeline.NormalizedTime
      Scale Curve=&MakeCustomFloatFromBool(Bool=$User.bisHeadHidden, True Float=0, False Float=1)
     ))
    ScaleA=true
    ScaleRGB=true
   ) {module=ScaleColor001}
   &ScaleMeshSize(Scale Factor=&MakeCustomVectorFromBool(
     Bool=$User.bisHeadHidden
     True Vector=<0 0 0>
     False Vector=&Multiply_Vector(
      A=&FN_FrontEndVector3Modifier(Gameplay=<1 1 1>, Front End=<0.5 1 1>)
      B=&VectorFromCurve(
       VectorCurve=DataInterface() {propertyValues {
         XCurve {keys [
           {arriveTangent=1.9574, interpMode=RCIM_Cubic, leaveTangent=1.9574, tangentMode=RCTM_User, time=0, value=1}
           {interpMode=RCIM_Cubic, tangentMode=RCTM_User, time=0.75, value=1.5}
           {interpMode=RCIM_Cubic, tangentMode=RCTM_User, time=1, value=1}
          ]}
         YCurve {keys [<at 0 1>,<at 1 1>]}
         ZCurve {keys [<at 0 1>,<at 1 1>]}
        }}
       CurveIndex=$System.Timeline.NormalizedTime
      )
     )
    ))
  ]
  rendererClasses [&NiagaraMeshRendererProperties]
  renderers [
   {rendererClass=&NiagaraMeshRendererProperties, rendererIndex=0}
   {rendererClass=&NiagaraMeshRendererProperties, rendererIndex=1}
   {rendererClass=&NiagaraMeshRendererProperties, rendererIndex=2}
   {rendererClass=&NiagaraMeshRendererProperties, rendererIndex=3}
  ]
 }
}
summary {
 systemName=NS_SpacePlunge_Head
 emitters [
  {emitterName=Glows, rendererClasses [&NiagaraSpriteRendererProperties, &NiagaraLightRendererProperties]}
  {emitterName=Hunt_Lightning, rendererClasses [&NiagaraMeshRendererProperties]}
  {emitterName=Hunt_MeshFX, rendererClasses [&NiagaraMeshRendererProperties]}
  {emitterName=CubicFlame, rendererClasses [&NiagaraSpriteRendererProperties]}
 ]
 userVariables [
  User.bIsFrontEnd {type=NiagaraBool}
  User.bIsFrontEndPreview {type=NiagaraBool}
  User.bisHeadHidden {type=NiagaraBool}
  User.IsReactive {type=NiagaraFloat, defaultValue=1, description=""}
  User.IsSuperLevel {type=NiagaraFloat, defaultValue=0, description=""}
  User.IsVisorHidden {type=NiagaraFloat, defaultValue=0, description=""}
  User.KillCount {type=NiagaraInt32}
  User.SuperLevelColor {type=LinearColor, description="", defaultValue {struct=LinearColor, value=<rgba 1 1 1 1>}}
  User.Variant {type=NiagaraFloat, defaultValue=0, description=""}
 ]
}
