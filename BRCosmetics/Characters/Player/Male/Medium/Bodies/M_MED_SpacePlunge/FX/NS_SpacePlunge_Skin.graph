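# Niagara system for the SpacePlunge character skin, adding glows, cubic flames, runes, lightning meshes and drifting idle cubes to the body.
# Emitters sample the skeletal mesh, several using particle partitions to pick among three bones or sockets.
# Variant index gates spawning and selects colors, while a system timeline drives flame, rune and lightning rates plus glow brightness.
# Glows include a masked particle light and hide with the head; front-end and time-of-day modifiers adjust brightness.
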NiagaraSystem /BRCosmetics/Characters/Player/Male/Medium/Bodies/M_MED_SpacePlunge/FX/NS_SpacePlunge_Skin
dependencies {
 usedDataInterfaces [
  &NiagaraDataInterfaceArrayFloat
  &NiagaraDataInterfaceArrayFloat2
  &NiagaraDataInterfaceCurve
  &NiagaraDataInterfaceVector2DCurve
  &NiagaraDataInterfaceArrayFloat3
  &NiagaraDataInterfaceArrayInt32
  &NiagaraDataInterfaceArrayColor
  &NiagaraDataInterfaceSpriteRendererInfo
  &NiagaraDataInterfaceMeshRendererInfo
  &NiagaraDataInterfaceVectorCurve
  &NiagaraDataInterfaceSkeletalMesh
  &NiagaraDataInterfaceColorCurve
 ]
 usedDynamicInputs [
  &SetBoolByFloatComparison
  &MakeIntFromFloat
  &SetBoolByIntComparison
 ]
 usedModules [
  &self:SystemSpawnScript.SystemScriptSource.SystemScriptGraph.NiagaraNodeAssignment_0.SetVariables_#1
  &self:SystemSpawnScript.SystemScriptSource.SystemScriptGraph.NiagaraNodeAssignment_2.SetVariables_#2
  &self:SystemSpawnScript.SystemScriptSource.SystemScriptGraph.NiagaraNodeAssignment_5.SetVariables_#3
  &self:SystemSpawnScript.SystemScriptSource.SystemScriptGraph.NiagaraNodeAssignment_3.SetVariables_#4
  &SystemState
  &BlendSpaceHelper
  &self:SystemSpawnScript.SystemScriptSource.SystemScriptGraph.NiagaraNodeAssignment_1.SetVariables_#5
  &self:SystemSpawnScript.SystemScriptSource.SystemScriptGraph.NiagaraNodeAssignment_4.SetVariables_#6
  /Game/Effects/Niagara/Modules/Logic/UpdateValueStateChange
  &Timeline
 ]
}
emitters {
 CubicFlame {
  emitterUpdateScript [
   &EmitterState()
   &SpawnRate(
    Spawn Probability=&SelectFloatFromArray(Direct Array Index=$System.VariantIndex, Float Selection Array=DataInterface() {propertyValues {FloatData [0,0,1,0,0]}})
    SpawnRate=&FloatFromCurve(
     FloatCurve=DataInterface() {propertyValues {Curve {keys [
         {interpMode=RCIM_Cubic, tangentMode=RCTM_User, time=0, value=0.25}
         {interpMode=RCIM_Cubic, tangentMode=RCTM_User, time=0.05, value=1}
         {interpMode=RCIM_Cubic, tangentMode=RCTM_User, time=0.75, value=1}
         {arriveTangent=-5.7829, tangentMode=RCTM_User, time=1, value=0.25}
        ]}}}
     CurveIndex=$System.Timeline.NormalizedTime
     Scale Curve=150
    )
    Use Spawn Probability=true
   )
   &SpawnPerFrame(
    Spawn:NiagaraBool=&Boolean_AndOperation(
     Bool1=&Boolean_AndOperation(
      Bool1=&Boolean_NotOperation(Bool=$System.UpdateValueStateChange.Changed)
      Bool2=&Boolean_NotOperation(Bool=$Transient.FirstFrame)
     )
     Bool2=$System.Reactive
    )
    Spawn Count=200
    Use Spawn Probability=false
   ) {enabled=false}
  ]
  particleSpawnScript [
   &PartitionParticles(Number Of Partitions=3)
   &InitializeParticle(
    Color Mode=::NewEnumerator0
    Lifetime=&RandomRangeFloat(Minimum=0.5, Maximum=0.75)
    Sprite Size Max=<15 20>
    Sprite Size Min=<10 10>
    Sprite Size Mode=::NewEnumerator2
    UsePositionOffset=false
    Write Material Random=true
   )
   &AddVelocity(Module.Use Falloff Distance=false, Origin Offset=<0 0 -10>, Override Seed=false, Velocity Mode=::NewEnumerator1, Velocity Speed=10) {enabled=false}
   &ApplyInitialForces() {enabled=false}
   &DynamicMaterialParameters(
    Index 0 Param 1=&RandomRangeFloat(Minimum=0, Maximum=1)
    Index 0 Param 2=&RandomRangeFloat(Minimum=0, Maximum=1)
    Module.Index0Float0Write=true
    Module.Index0Float1Write=true
    Module.Index0Float2Write=true
    Module.Index0Float3Write=true
   )
  ]
  particleUpdateScript [
   &ParticleState()
   &ScaleSpriteSize(
    Non-Uniform Curve Scale=&RandomRangeVector2D(Minimum=<0.8 0.8>, Maximum=<1 1>)
    Non-Uniform Curve Sprite Scale=DataInterface() {propertyValues {
      XCurve {keys [{interpMode=RCIM_Cubic, tangentMode=RCTM_User, time=0, value=1}, {arriveTangent=-2.4882, tangentMode=RCTM_User, time=1, value=0}]}
      YCurve {keys [
        {arriveTangent=2.1941, interpMode=RCIM_Cubic, leaveTangent=2.1941, tangentMode=RCTM_User, time=0, value=0.1}
        {arriveTangent=0.6251, interpMode=RCIM_Cubic, leaveTangent=0.6251, tangentMode=RCTM_User, time=1, value=2}
       ]}
     }}
    Scale Sprite Size Mode=::NewEnumerator3
   )
   &CameraOffset(Camera Offset Amount=-2)
   &SkeletalMeshLocation(
    Bone / Socket Index=$Particles.PartitionParticles.Partition
    Bone Sampling Mode=::NewEnumerator4
    MaskSpawnGroup=false
    Orientation Sampling=::NewEnumerator1
    Sampled Position Offset=&SelectVectorFromArray(
     Direct Array Index=$Particles.PartitionParticles.Partition
     Vector Selection Array=DataInterface() {propertyValues {FloatData [<0 0 0>, <0 0 0>, <0 5 0>]}}
    )
    Skeletal Mesh=$System.Skeletal Mesh
   )
   &ShapeLocation(
    Box Midpoint=<0.5 0.5 0.75>
    Box Size=&SelectVectorFromArray(
     Direct Array Index=$Particles.PartitionParticles.Partition
     Vector Selection Array=DataInterface() {propertyValues {FloatData [<15 15 15>, <15 15 15>, <25 20 20>]}}
    )
    MaskSpawnGroup=false
    Override Seed=false
    Shape Origin=$Particles.Position
    Shape Primitive=::NewEnumerator2
    Surface Only=false
    UseOwnerScale=true
   )
   &GravityForce(Gravity=&RandomRangeVector(Minimum=<0 0 10>, Maximum=<0 0 25>)) {enabled=false}
   &SolveForcesAndVelocity(Clamp Velocity=false, Limit Acceleration=false)
   &ScaleColor(
    Scale Alpha=&FloatFromCurve(FloatCurve=DataInterface() {propertyValues {Curve {keys [<at 0 0>,<at 0.1 1>,<at 1 0>]}}})
    Scale RGB=&Multiply_VectorByFloat(
     Vector=<1 1 1>
     Float=&FloatFromCurve(FloatCurve=DataInterface() {propertyValues {Curve {keys [<at 0 1>,<at 0.084 3>,<at 1 0>]}}})
    )
    ScaleA=true
    ScaleRGB=true
   )
  ]
  rendererClasses [&NiagaraSpriteRendererProperties]
  renderers [{rendererClass=&NiagaraSpriteRendererProperties}]
 }
 Glows {
  emitterUpdateScript [
   &EmitterState(
    Inactive Response=::NewEnumerator1
    Life Cycle Mode=::NewEnumerator1
    Loop Behavior=::NewEnumerator1
    Loop Duration Mode=::NewEnumerator1
    UseLoopDelay=false
   )
   &SpawnBurst_Instantaneous(
    Spawn Count=3
    Spawn Probability=&SelectFloatFromArray(Direct Array Index=$System.VariantIndex, Float Selection Array=DataInterface() {propertyValues {FloatData [0,0,1,0,0]}})
    Use Spawn Probability=true
    UseLoopCountLimit=true
   )
  ]
  particleSpawnScript [
   &PartitionParticles(Number Of Partitions=3)
   &InitializeParticle(
    Color=$System.VariantColorLinear
    Sprite Size=&SelectVector2DFromArray(
     Direct Array Index=$Particles.PartitionParticles.Partition
     Vector 2D Selection Array=DataInterface() {propertyValues {FloatData [<60 60>, <60 60>, <150 200>]}}
    )
    Sprite Size Mode=::NewEnumerator4
    UsePositionOffset=false
    Write Material Random=true
   )
  ]
  particleUpdateScript [
   &ParticleState(Kill Particles When Lifetime Has Elapsed=false, Let Infinitely Lived Particles Die When Emitter Deactivates=true)
   &ScaleSpriteSize(
    Scale Sprite Size Mode=::NewEnumerator0
    Uniform Scale Factor=&MakeCustomFloatFromBool(Bool=$User.bisHeadHidden, True Float=0, False Float=1)
   )
   &ScaleColor(
    Scale Alpha=&MakeCustomFloatFromBool(Bool=$User.bisHeadHidden, True Float=0, False Float=1)
    Scale RGB=&VectorFromFloat(Value=&FloatFromCurve(
      FloatCurve=DataInterface() {propertyValues {Curve {keys [
          <at 0 1>
          {interpMode=RCIM_Cubic, tangentMode=RCTM_User, time=0.0368, value=1}
          <at 0.0666 3>
          {interpMode=RCIM_Cubic, tangentMode=RCTM_User, time=0.75, value=3}
          {arriveTangent=-16.1427, tangentMode=RCTM_User, time=1, value=1}
         ]}}}
      CurveIndex=$System.Timeline.NormalizedTime
      Scale Curve=&FN_FrontEndFloatModifier(Gameplay=1, Front End=0.4)
     ))
    ScaleA=true
    ScaleRGB=true
   )
   &SkeletalMeshLocation(
    Bone / Socket Index=$Particles.PartitionParticles.Partition
    Bone Sampling Mode=::NewEnumerator4
    MaskSpawnGroup=false
    Skeletal Mesh=$System.Skeletal Mesh
   )
   &CameraOffset(Camera Offset Amount=-2)
   &FN_Light_Attributes(
    Color=$System.VariantColorLinear
    Create Color Attribute=true
    Create Exponent Attribute=true
    Create Position Attribute=true
    Create Radius Attribute=true
    Enable Particle Light Enable=true
    Enable Visibility Tag=true
    Exponent=2
    Light Enabled:NiagaraBool=/Niagara/DynamicInputs/Helpers/MaskBool(
     Mask Source:NiagaraInt32=1
     Mask Parameter:NiagaraInt32=$Particles.PartitionParticles.Partition
     Mask:NiagaraInt32=2
     Masked Bool:NiagaraBool=true
     Passthrough Bool:NiagaraBool=false
    )
    Position=&AddVectorToPosition(Position=$Particles.Position, Vector=<0 20 0>)
    Radius=40
    Scale Color=5
    VisibilityTag=1
   )
  ]
  rendererClasses [&NiagaraSpriteRendererProperties, &NiagaraLightRendererProperties]
  renderers [
   {rendererClass=&NiagaraSpriteRendererProperties}
   {rendererClass=&NiagaraLightRendererProperties, rendererIndex=1}
  ]
 }
 Hunt_Lightning {
  emitterUpdateScript [
   &EmitterState()
   &SpawnRate(
    Spawn Probability=/Niagara/DynamicInputs/Helpers/DistributionCurveFloat(
     Minimum:NiagaraFloat=0.5
     Maximum:NiagaraFloat=1
     Distribution Curve:NiagaraDataInterfaceCurve=DataInterface() {propertyValues {Curve {
        defaultValue=0
        keys [{interpMode=RCIM_Cubic, tangentMode=RCTM_User, time=0.5, value=0}, {arriveTangent=-0.0819, tangentMode=RCTM_User, time=1, value=1}]
       }}}
     Randomness Mode:ENiagaraRandomnessMode=::NewEnumerator0
     Evaluation Type:ENiagaraRandomnessEvaluation=::NewEnumerator1
     Seed:NiagaraInt32=0
     Fixed Random Seed:NiagaraBool=false
    )
    SpawnRate=&MaskFloat(
     Mask Source=1
     Mask Parameter=$System.VariantIndex
     Mask=2
     Masked Float=&FloatFromCurve(
      FloatCurve=DataInterface() {propertyValues {Curve {keys [
          {interpMode=RCIM_Cubic, tangentMode=RCTM_User, time=0, value=0}
          {interpMode=RCIM_Cubic, tangentMode=RCTM_User, time=0.05, value=1}
          {interpMode=RCIM_Cubic, tangentMode=RCTM_User, time=0.75, value=1}
          {arriveTangent=-5.7829, tangentMode=RCTM_User, time=1, value=0}
         ]}}}
      CurveIndex=$System.Timeline.NormalizedTime
      Scale Curve=100
     )
     Passthrough Float=0
    )
    Use Spawn Probability=true
   )
  ]
  particleSpawnScript [
   &PartitionParticles(Number Of Partitions=3)
   &InitializeParticle(
    Color=$System.VariantColorLinear
    Lifetime Max=0.5
    Lifetime Min=0.25
    Lifetime Mode=::NewEnumerator1
    Mesh Scale=&Multiply_VectorByFloat(
     Vector=&Lerp_Vector(
      A=&RandomRangeVector(Minimum=<0.5 0.4 0.4>, Maximum=<0.75 0.6 0.6>)
      B=&RandomRangeVector(Minimum=<0.8 1 1>, Maximum=<1.25 2 2>)
      Alpha=&FloatFromCurve(
       FloatCurve=DataInterface() {propertyValues {Curve {keys [{interpMode=RCIM_Cubic, tangentMode=RCTM_User, time=0.5, value=0}, <at 1 1>]}}}
       CurveIndex=&RandomRangeFloat(Minimum=0, Maximum=1)
      )
     )
     Float=&SelectFloatFromArray(Direct Array Index=$Particles.PartitionParticles.Partition, Float Selection Array=DataInterface() {propertyValues {FloatData [0.2,0.35]}})
    )
    Mesh Scale Mode=::NewEnumerator4
    UsePositionOffset=true
    Write Material Random=true
   )
   &InitialMeshOrientation(
    Mesh Orientation Mode=::NewEnumerator2
    Orientation Coordinate Space=::Local
    Orientation Vector=&SelectVectorFromArray(
     Direct Array Index=$Particles.PartitionParticles.Partition
     Vector Selection Array=DataInterface() {propertyValues {FloatData [<1 -1 0>, <-1 0 0>, <1 0 0>]}}
    )
    Rotation=&RandomRangeVector(Minimum=<-0.25 -0.25 -0.25>, Maximum=<0.25 0.25 0.25>)
    Use Rotation Vector=true
   ) {module=InitialMeshOrientation002}
   /SetVariables(Particles.Random=&RandomRangeFloat(Minimum=0, Maximum=1)) {isSetParameters=true}
  ]
  particleUpdateScript [
   &SkeletalMeshLocation(
    Bone / Socket Index=&SelectIntFromArray(Direct Array Index=$Particles.PartitionParticles.Partition, Int Selection Array=DataInterface() {propertyValues {IntData [0,1,2]}})
    Bone Sampling Mode=::NewEnumerator4
    MaskSpawnGroup=false
    Orientation Sampling=::NewEnumerator1
    Sampled Position Offset=&SelectVectorFromArray(
     Direct Array Index=$Particles.PartitionParticles.Partition
     Vector Selection Array=DataInterface() {propertyValues {FloatData [<6 -2 0>, <-5 0 0>]}}
    )
    Skeletal Mesh=$System.Skeletal Mesh
   )
   &ShapeLocation(
    MaskSpawnGroup=false
    Override Seed=false
    Rotation Mode=::NewEnumerator1
    Shape Origin=$Particles.Position
    Sphere Radius=5
    UseOwnerScale=true
   )
   &ParticleState()
   &ScaleMeshSize()
   &ScaleColor(
    Scale Alpha=&Multiply_Float(
     A=10
     B=&FloatFromCurve(FloatCurve=DataInterface() {propertyValues {Curve {keys [
          {interpMode=RCIM_Cubic, tangentMode=RCTM_User, time=0, value=0}
          {interpMode=RCIM_Cubic, tangentMode=RCTM_User, time=0.1, value=0}
          {interpMode=RCIM_Cubic, tangentMode=RCTM_User, time=0.2, value=0.8}
          {interpMode=RCIM_Cubic, tangentMode=RCTM_User, time=1, value=0}
         ]}}})
    )
    Scale RGB=&Multiply_VectorByFloat(
     Vector=&Multiply_VectorByFloat(Vector=<1 1 1>, Float=&FN_FrontEndFloatModifier(Gameplay=1, Front End=0.5))
     Float=&Multiply_Float(
      A=&RandomRangeFloat(Minimum=0.5, Maximum=1, Evaluation Type=::NewEnumerator1)
      B=&Multiply_Float(A=&FNTimeOfDayScalar(Night Scale=0.2, Morning Scale=1, Evening Scale=1), B=10)
     )
    )
    ScaleA=true
    ScaleRGB=true
   )
   &DynamicMaterialParameters(
    Index 0 Param 1=&FloatFromCurve(FloatCurve=DataInterface() {propertyValues {Curve {keys [{tangentMode=RCTM_User, time=0, value=1}, {tangentMode=RCTM_User, time=1, value=1}]}}})
    Index 0 Param 2=0.7
    Index 0 Param 3=&FloatFromCurve(FloatCurve=DataInterface() {propertyValues {Curve {keys [
         {arriveTangent=4.0072, interpMode=RCIM_Cubic, leaveTangent=4.0072, tangentMode=RCTM_User, time=0, value=0}
         {arriveTangent=2.7083, interpMode=RCIM_Cubic, leaveTangent=2.7083, time=0.2543, value=0.6225}
         {arriveTangent=1.8473, interpMode=RCIM_Cubic, leaveTangent=1.8473, time=0.6, value=1.625}
         {arriveTangent=1.6993, interpMode=RCIM_Cubic, leaveTangent=1.6993, tangentMode=RCTM_User, time=1, value=2}
        ]}}})
    Index 0 Param 4=5
    Index 1 Param 1=&MaskFloat(Mask Source=1, Mask Parameter=$System.VariantIndex, Mask=2, Masked Float=0.4, Passthrough Float=0.15)
    Index 1 Param 2=$Particles.Random
    Index 1 Param 3=1.3
    Index 1 Param 4=1
    Index 2 Param 1=&FloatFromCurve(FloatCurve=DataInterface() {propertyValues {Curve {keys [<at 0 1>,<at 1 0.3>]}}})
    Module.Index0Float0Write=true
    Module.Index0Float1Write=true
    Module.Index0Float2Write=true
    Module.Index0Float3Write=true
    Module.Index1Float0Write=true
    Module.Index1Float1Write=true
    Module.Index1Float2Write=true
    Module.Index1Float3Write=true
    Module.Index2Float0Write=true
    Module.Index2Float1Write=true
    Module.Index2Float2Write=true
    Module.Index2Float3Write=true
    Write Parameter Index 1=::NewEnumerator1
    Write Parameter Index 2=::NewEnumerator1
   )
   &SolveForcesAndVelocity(Clamp Velocity=false, Limit Acceleration=false)
   /SetVariables(Particles.Lightning glow color:LinearColor=&MakeLinearColorFromVectorAndFloat(
     "Vector (RGB)"=&MakeVectorFromLinearColorRGB(Color=&MultiplyLinearColor(
       Linear Color=&SelectLinearColorFromArray(
        Array Sampling Mode=::NewEnumerator1
        Direct Array Index=$System.VariantIndex
        Color Selection Array=DataInterface() {propertyValues {ColorData [<rgba 4 1.6071 0.1968 1>, <rgba 0.3366 2.0701 4 1>, <rgba 4 1.6071 0.1968 1>]}}
        Random Seed=?
       )
       Scale Factor=&MakeVector4_FromFloat(Float=&RandomRangeFloat(Minimum=0.5, Maximum=1, Evaluation Type=::NewEnumerator1))
      ))
     "Float (Alpha)"=&Multiply_Float(
      A=&FloatFromCurve(FloatCurve=DataInterface() {propertyValues {Curve {keys [<at 0 1 RCIM_Cubic>, {interpMode=RCIM_Cubic, leaveTangent=-3.6307, time=0.9, value=0}]}}})
      B=1
     )
    )) {isSetParameters=true}
   /SetVariables(Particles.Glow Dy02:Vector4f=&MakeVector4(
     X=&FloatFromCurve(FloatCurve=DataInterface() {propertyValues {Curve {keys [{tangentMode=RCTM_User, time=0, value=0.2}, {tangentMode=RCTM_User, time=0.5, value=1}]}}})
     Y=0.7
     Z=0
     W=5
    )) {isSetParameters=true}
   /SetVariables(Particles.Glow Dy03:Vector4f=&MakeVector4(
     X=&FloatFromCurve(
      FloatCurve=DataInterface() {propertyValues {Curve {keys [<at 0 1>,<at 1 0.5>]}}}
      Scale Curve=&MaskFloat(Mask Source=1, Mask Parameter=$System.VariantIndex, Mask=2, Masked Float=0.7, Passthrough Float=0.5)
     )
     Y=$Particles.Random
     Z=1.3
     W=1
    )) {isSetParameters=true}
  ]
  rendererClasses [&NiagaraMeshRendererProperties]
  renderers [
   {rendererClass=&NiagaraMeshRendererProperties, rendererIndex=0}
   {rendererClass=&NiagaraMeshRendererProperties, rendererIndex=1}
  ]
 }
 IdleCubes {
  emitterUpdateScript [
   &EmitterState()
   &SpawnRate(
    Spawn Probability=&MaskFloat(Mask Source=1, Mask Parameter=$System.VariantIndex, Mask=2, Masked Float=1, Passthrough Float=0)
    SpawnRate=7
    Use Spawn Probability=true
   )
  ]
  particleSpawnScript [
   &InitializeParticle(
    Color Maximum=<rgba 1 0.8541 0.8156 1>
    Color Minimum=<rgba 0.6094 0.5522 0.4975 1>
    Color Mode=::NewEnumerator2
    Lifetime=&RandomRangeFloat(Minimum=12, Maximum=16)
    Mesh Scale Mode=::NewEnumerator3
    Mesh Uniform Scale=&RandomRangeFloat(Minimum=0.6, Maximum=0.8)
    UsePositionOffset=false
    Write Material Random=true
   )
   /SetVariables() {enabled=false, isSetParameters=true}
   &SubUVAnimation(
    Mesh Renderer=DataInterface() {propertyValues {MeshRenderer=&self:IdleCubes_0.NiagaraMeshRendererProperties_0}}
    Output Alternate Renderer Binding=true
    Random Start Frame=true
    Renderer Type:ENiagaraMeshOrSprite=::NewEnumerator1
    SubUV Lookup Index Scale=0
    UseEndFrame=false
    UseStartFrame=false
   )
   &InitialMeshOrientation(Mesh Orientation Mode=::NewEnumerator7, Use Rotation Vector=false)
  ]
  particleUpdateScript [
   &ParticleState()
   &SkeletalMeshLocation(
    Bone / Socket Index=2
    Bone Sampling Mode=::NewEnumerator4
    MaskSpawnGroup=false
    Sampled Orientation Coordinate Space=::Local
    Sampled Position Offset=&RandomRangeVector(Minimum=<-8 -11 -12>, Maximum=<7 10 12>)
    Skeletal Mesh=$System.Skeletal Mesh
   )
   &ShapeLocation(
    MaskSpawnGroup=false
    Override Seed=false
    Rotation Mode=::NewEnumerator5
    Shape Origin=$Particles.Position
    Sphere Radius=15
    UseOwnerScale=true
   ) {enabled=false}
   &ScaleMeshSize(Scale Factor=&VectorFromCurve(
     VectorCurve=DataInterface() {propertyValues {
       XCurve {keys [<at 0 1>,<at 0.3426 0.9144>,<at 0.9758 1.3967>]}
       YCurve {keys [<at 0 1>,<at 0.6063 1.3611>,<at 0.9654 1.1803>]}
       ZCurve {keys [<at 0 0.577>,<at 0.1533 1.3432>,<at 0.6073 1.0575>,<at 0.9862 1.2393>]}
      }}
     Scale Curve=&Multiply_VectorByFloat(
      Vector=&RandomRangeVector(Minimum=<0.75 0.75 0.75>, Maximum=<1.25 1.25 1.25>)
      Float=&FloatFromCurve(FloatCurve=DataInterface() {propertyValues {Curve {keys [
           {interpMode=RCIM_Cubic, tangentMode=RCTM_User, time=0, value=0.25}
           {interpMode=RCIM_Cubic, tangentMode=RCTM_User, time=0.5, value=1}
           {arriveTangent=-4.1587, interpMode=RCIM_Cubic, leaveTangent=-4.1587, tangentMode=RCTM_User, time=1, value=-0.0118}
          ]}}})
     )
    ))
   &DynamicMaterialParameters(
    Index 0 Param 1=&FloatFromCurve(FloatCurve=DataInterface() {propertyValues {Curve {keys [<at 0 0>,<at 1 4>]}}}, Scale Curve=2)
    Index 0 Param 2=1
    Index 0 Param 3=&FloatFromCurve(
     FloatCurve=DataInterface() {propertyValues {Curve {keys [{interpMode=RCIM_Cubic, tangentMode=RCTM_User, time=0, value=2}, {arriveTangent=-2.5233, tangentMode=RCTM_User, time=1.0059, value=1}]}}}
     Scale Curve=&Waveform(
      "Waveform [1]"=::NewEnumerator9
      "[1] Input Position"=&RandomRangeFloat(Minimum=0, Maximum=1)
      "[1] Frequency"=3
      "[1] Amplitude Min/Max"=<0.75 1>
     )
    )
    Module.Index0Float0Write=true
    Module.Index0Float1Write=true
    Module.Index0Float2Write=true
    Module.Index0Float3Write=true
   )
   &ScaleColor(
    Scale RGB=&Multiply_VectorByFloat(
     Vector=<1 1 1>
     Float=&FloatFromCurve(FloatCurve=DataInterface() {propertyValues {Curve {keys [<at -0.0926 1>,<at -0.0013 1>,<at 0.2854 1>,<at 0.9021 1>,<at 1 1>]}}})
    )
    ScaleA=true
    ScaleRGB=true
   )
   /SetVariables(Particles.Position=&AddVectorToPosition(
     Position=$Particles.Position
     Vector=&VectorFromCurve(
      VectorCurve=DataInterface() {propertyValues {
        XCurve {keys [
          {arriveTangent=3.0078, interpMode=RCIM_Cubic, leaveTangent=3.0078, tangentMode=RCTM_User, time=0, value=0}
          {interpMode=RCIM_Cubic, tangentMode=RCTM_User, time=0.5, value=1}
          {arriveTangent=-3.9312, interpMode=RCIM_Cubic, leaveTangent=-3.9312, tangentMode=RCTM_User, time=1, value=2}
         ]}
        YCurve {keys [
          {arriveTangent=3.0078, interpMode=RCIM_Cubic, leaveTangent=3.0078, tangentMode=RCTM_User, time=0, value=0}
          {interpMode=RCIM_Cubic, tangentMode=RCTM_User, time=0.5, value=1}
          {arriveTangent=-3.9312, interpMode=RCIM_Cubic, leaveTangent=-3.9312, tangentMode=RCTM_User, time=1, value=2}
         ]}
        ZCurve {keys [
          {arriveTangent=3.0078, interpMode=RCIM_Cubic, leaveTangent=3.0078, tangentMode=RCTM_User, time=0, value=0}
          {interpMode=RCIM_Cubic, tangentMode=RCTM_User, time=0.5, value=1}
          {arriveTangent=-3.9312, interpMode=RCIM_Cubic, leaveTangent=-3.9312, tangentMode=RCTM_User, time=1, value=2}
         ]}
       }}
      Scale Curve=&Multiply_VectorByFloat(Vector=&RandomRangeVector(Minimum=<-1 -1 -1>, Maximum=<1 1 1>), Float=3)
     )
    )) {isSetParameters=true}
   &CameraOffset(Camera Offset Amount=10) {enabled=false}
  ]
  rendererClasses [&NiagaraSpriteRendererProperties, &NiagaraMeshRendererProperties]
  renderers [
   {rendererClass=&NiagaraSpriteRendererProperties}
   {rendererClass=&NiagaraMeshRendererProperties, rendererIndex=1}
  ]
 }
 Runes {
  emitterUpdateScript [
   &EmitterState()
   &SpawnRate(
    Spawn Probability=&SelectFloatFromArray(Direct Array Index=$System.VariantIndex, Float Selection Array=DataInterface() {propertyValues {FloatData [0,0,1,0,0]}})
    SpawnRate=&FloatFromCurve(
     FloatCurve=DataInterface() {propertyValues {Curve {keys [
         {interpMode=RCIM_Cubic, tangentMode=RCTM_User, time=0, value=0.25}
         {interpMode=RCIM_Cubic, tangentMode=RCTM_User, time=0.05, value=1}
         {interpMode=RCIM_Cubic, tangentMode=RCTM_User, time=0.75, value=1}
         {arriveTangent=-5.7829, tangentMode=RCTM_User, time=1, value=0.25}
        ]}}}
     CurveIndex=$System.Timeline.NormalizedTime
     Scale Curve=60
    )
    Use Spawn Probability=true
   )
  ]
  particleSpawnScript [
   &InitializeParticle(
    Lifetime=&RandomRangeFloat(Minimum=0.5, Maximum=1.25)
    Sprite Size Mode=::NewEnumerator1
    Uniform Sprite Size Max=4
    Uniform Sprite Size Min=2
    UsePositionOffset=false
    Write Material Random=true
   )
   &SkeletalMeshLocation(Bone Sampling Mode=::NewEnumerator0, MaskSpawnGroup=false, OverrideRandomSeed=false, Skeletal Mesh=$System.Skeletal Mesh)
   &ShapeLocation(MaskSpawnGroup=false, Offset Mode=::NewEnumerator0, Override Seed=false, Sphere Radius=8, UseOwnerScale=true)
   &AddVelocity(
    Module.Use Falloff Distance=false
    Override Seed=false
    Velocity Mode=::NewEnumerator1
    Velocity Speed=&RandomRangeFloat(Minimum=10, Maximum=15)
   )
  ]
  particleUpdateScript [
   &ParticleState()
   &DynamicMaterialParameters(
    Index 0 Param 1=&Waveform(
     "Waveform [1]"=::NewEnumerator9
     "[1] Phase Scale"=&RandomRangeFloat(Minimum=0, Maximum=1)
     "[1] Frequency"=&RandomRangeFloat(Minimum=6, Maximum=10)
     "[1] Amplitude Min/Max"=<0 4>
    )
    Index 0 Param 2=1
    Index 0 Param 3=1
    Module.Index0Float0Write=true
    Module.Index0Float1Write=true
    Module.Index0Float2Write=true
    Module.Index0Float3Write=true
   )
   &SubUVAnimation(
    Random Start Frame=true
    "Sprite Renderer "=DataInterface() {propertyValues {SpriteRenderer=&self:Runes_0.NiagaraSpriteRendererProperties_0}}
    SubUV Lookup Index:NiagaraFloat=&Add_Float(
     A=&Multiply_Float(A=$Engine.System.Age, B=&RandomRangeFloat(Minimum=0.05, Maximum=0.2))
     B=&RandomRangeFloat(Minimum=0, Maximum=1)
    )
    SubUV Lookup Index Scale:NiagaraFloat=&FloatFromCurve(
     FloatCurve=DataInterface() {propertyValues {Curve {keys [<at 0 4>,<at 0.1 1>,<at 1 1>]}}}
     Scale Curve=&RandomRangeFloat(Minimum=1, Maximum=2)
    )
    UseEndFrame=false
    UseStartFrame=false
   ) {module=SubUVAnimation001}
   &ScaleColor(
    Scale Alpha=&FloatFromCurve(FloatCurve=DataInterface() {propertyValues {Curve {keys [{interpMode=RCIM_Cubic, tangentMode=RCTM_User, time=0, value=1}, <at 1 0>]}}})
    Scale RGB=&Multiply_VectorByFloat(
     Vector=<1 1 1>
     Float=&FloatFromCurve(
      FloatCurve=DataInterface() {propertyValues {Curve {keys [<at 0 1>, {interpMode=RCIM_Cubic, tangentMode=RCTM_User, time=0.2, value=3}, <at 0.4 1>, <at 1 1>]}}}
      Scale Curve=2
     )
    )
    ScaleA=true
    ScaleRGB=true
   )
   &GravityForce() {enabled=false}
   &Drag(Use Linear Drag=true, Use Rotational Drag=false) {enabled=false}
   &SolveForcesAndVelocity(Clamp Velocity=false, Limit Acceleration=false)
   &BlendSpace("Position Blend (World <-> Local)":NiagaraFloat=&FloatFromCurve(
     FloatCurve=DataInterface() {propertyValues {Curve {keys [{interpMode=RCIM_Cubic, tangentMode=RCTM_User, time=0.6, value=1}, {arriveTangent=-1.4851, tangentMode=RCTM_User, time=0.7, value=0}]}}}
     Translation Contribution:NiagaraFloat=1
     Rotation Contribution:NiagaraFloat=1
    ))
  ]
  rendererClasses [&NiagaraSpriteRendererProperties]
  renderers [{rendererClass=&NiagaraSpriteRendererProperties}]
 }
}
summary {
 systemName=NS_SpacePlunge_Skin
 emitters [
  {emitterName=Glows, rendererClasses [&NiagaraSpriteRendererProperties, &NiagaraLightRendererProperties]}
  {emitterName=CubicFlame, rendererClasses [&NiagaraSpriteRendererProperties]}
  {emitterName=Hunt_Lightning, rendererClasses [&NiagaraMeshRendererProperties]}
  {emitterName=Runes, rendererClasses [&NiagaraSpriteRendererProperties]}
  {emitterName=IdleCubes, rendererClasses [&NiagaraSpriteRendererProperties, &NiagaraMeshRendererProperties]}
 ]
 userVariables [
  User.bIsFrontEnd {type=NiagaraBool}
  User.bIsFrontEndPreview {type=NiagaraBool}
  User.bisHeadHidden {type=NiagaraBool}
  User.IsReactive {type=NiagaraFloat, defaultValue=1, description=""}
  User.KillCount {type=NiagaraInt32}
  User.Variant {type=NiagaraFloat, defaultValue=0, description=""}
 ]
}
