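# Material function for the SpacePlunge player body that generates animated, two-tone storm-cloud smoke, outputting color, FresnelMask and SphereMask.
# Two noise layers, one on mesh UVs and one on rerouted coordinates, scroll cloud textures distorted by panning noise.
# View-dependent masks use the cloud normals, noise breakup and adjustable hardness.
# A virtual sphere built from camera and normal vectors drives bump-offset coordinates, suggesting depth inside the surface.
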MaterialFunction /BRCosmetics/Characters/Player/Male/Medium/Bodies/M_MED_SpacePlunge/Materials/MF_StormNoise_Smoke
graph {nodes [
  n1 = WorldPosition()
  n2 = ScalarParameter() {defaultValue=-12, parameterName=D}
  n3 = PixelNormalWS()
  n4 = CameraVectorWS()
  n5 = PixelNormalWS()
  n6 = Add(A=@n1.XYZ, B=Divide(A=Multiply(A=@n2, B=@n3), B=DotProduct(A=@n4, B=@n5)))
  n7 = Normalize(VectorInput=Subtract(A=CameraPositionWS(), B=@n6))
  n8 = Transform(None=@n7) {transformSourceType=TRANSFORMSOURCE_World, transformType=TRANSFORM_Tangent}
  n9 = Multiply(A=TextureCoordinate(), B=ScalarParameter() {defaultValue=0.7, group=StormNoise, parameterName=CloudTiling})
  Add(A=Multiply(A=ComponentMask(None=@n8) {g=true, r=true}, B=ScalarParameter() {defaultValue=1, parameterName=OffsetIntensity}), B=@n9)
  n10 = NamedRerouteUsage()
  n11 = NamedRerouteUsage()
  n12 = Power(Base=Saturate(None=NamedRerouteUsage())) {constExponent=1}
  FunctionOutput(None=LinearInterpolate(A=Multiply(A=@n10, B=@n11), B=NamedRerouteUsage(), Alpha=@n12))
  FunctionOutput(None=LinearInterpolate(A=NamedRerouteUsage(), B=NamedRerouteUsage(), Alpha=@n12)) {outputName=FresnelMask, sortPriority=1}
  FunctionOutput(None=NamedRerouteUsage()) {outputName=SphereMask, sortPriority=1}
  n13 = TextureCoordinate()
  n14 = VectorParameter() {group=StormNoise, parameterName=NoiseTiling, channelNames {a="", b="", g=Tile Y, r=Tile X}, defaultValue=<rgba 4 4 4 1>}
  n15 = ComponentMask(None=@n14) {g=true, r=true}
  n16 = ScalarParameter() {defaultValue=0.1, group=StormNoise, parameterName=DistortionStrength}
  n17 = NamedRerouteUsage()
  n18 = VectorParameter() {group=StormNoise, parameterName=NoiseSpeed, channelNames {a="", b="", g=Speed Y, r=Speed X}, defaultValue=<rgba 0.025 -0.05 1 1>}
  n19 = Add(A=Add(A=Multiply(A=@n13, B=@n15), B=Multiply(A=@n16, B=@n17)), B=Multiply(A=Time(), B=ComponentMask(None=@n18) {g=true, r=true}))
  n20 = TextureSampleParameter2D(UVs=@n19) {
   group=StormNoise
   parameterName=NoiseNormal
   samplerSource=SSM_Wrap_WorldGroupSettings
   samplerType=SAMPLERTYPE_Normal
   texture=/Game/Characters/Player/Male/Medium/Bodies/M_MED_Hurricane/Textures/T_MED_Hurricane_StormCloud_N
  }
  LinearInterpolate(A=@n20.RGB, B=NamedRerouteUsage(), Alpha=NamedRerouteUsage())
  Multiply()
  n21 = NamedRerouteUsage()
  n22 = VectorParameter() {group=StormNoise, parameterName=DistortionTiling, channelNames {a="", b="", g=Tile Y, r=Tile X}, defaultValue=<rgba 4 3 2 1>}
  n23 = ComponentMask(None=@n22) {g=true, r=true}
  n24 = VectorParameter() {group=StormNoise, parameterName=DistortionSpeed, channelNames {a="", b="", g=Speed Y, r=Speed X}, defaultValue=<rgba -0.1 0.1 1 1>}
  n25 = TextureSampleParameter2D(UVs=Panner(Coordinate=Multiply(A=@n21, B=@n23), Speed=ComponentMask(None=@n24) {g=true, r=true})) {
   group=StormNoise
   parameterName=DistortionTexture
   samplerSource=SSM_Wrap_WorldGroupSettings
   samplerType=SAMPLERTYPE_LinearColor
   texture=&T_NoisePack_LinearColor
  }
  NamedRerouteDeclaration(None=ChannelMaskParameter(None=@n25.RGB) {group=StormNoise, parameterName=DistortionMaskChannel}.R) {name=DistortionMask, nodeColor=<rgba 0 1 0.0859 1>}
  NamedRerouteDeclaration(None=ComponentMask(None=ConstantBiasScale(None=@n25.RGB) {bias=-0.5, scale=2}) {g=true, r=true}) {name=UVDistortion, nodeColor=<rgba 1 0 0.75 1>}
  n26 = FunctionInput(Preview=GetMaterialAttributes().MaterialAttributes) {bUsePreviewValueAsDefault=true, inputName=Material, inputType=FunctionInput_MaterialAttributes}
  NamedRerouteDeclaration(None=GetMaterialAttributes(None=NamedRerouteDeclaration(None=@n26) {name=InMaterial, nodeColor=<rgba 1 0.9375 0 1>}) {attributeGetTypes [#1]}.Normal) {name=InNormal, nodeColor=<rgba 0.7578 0 1 1>}
  n27 = NamedRerouteDeclaration(None=@n7) {name=NewVector, nodeColor=<rgba 0 0.2969 1 1>}
  n28 = Multiply(A=@n3, B=ScalarParameter() {defaultValue=-50, parameterName=CenterDist})
  n29 = Normalize(VectorInput=Subtract(A=@n6, B=Add(A=@n1.XYZ, B=@n28)))
  NamedRerouteDeclaration(None=Saturate(None=DotProduct(A=@n27, B=@n29))) {name=NewArea, nodeColor=<rgba 0 0.5547 1 1>}
  n30 = TextureCoordinate()
  n31 = VectorParameter() {group=StormNoise, parameterName=DistortionTiling, channelNames {a="", b="", g=Tile Y, r=Tile X}, defaultValue=<rgba 4 3 2 1>}
  n32 = ComponentMask(None=@n31) {g=true, r=true}
  n33 = VectorParameter() {group=StormNoise, parameterName=DistortionSpeed, channelNames {a="", b="", g=Speed Y, r=Speed X}, defaultValue=<rgba -0.1 0.1 1 1>}
  n34 = TextureSampleParameter2D(UVs=Panner(Coordinate=Multiply(A=@n30, B=@n32), Speed=ComponentMask(None=@n33) {g=true, r=true})) {
   group=StormNoise
   parameterName=DistortionTexture
   samplerSource=SSM_Wrap_WorldGroupSettings
   samplerType=SAMPLERTYPE_LinearColor
   texture=&T_NoisePack_LinearColor
  }
  NamedRerouteDeclaration(None=ChannelMaskParameter(None=@n34.RGB) {group=StormNoise, parameterName=DistortionMaskChannel}.R) {name=DistortionMask_1, nodeColor=<rgba 0 1 0.0859 1>}
  NamedRerouteDeclaration(None=ComponentMask(None=ConstantBiasScale(None=@n34.RGB) {bias=-0.5, scale=2}) {g=true, r=true}) {name=UVDistortion_1, nodeColor=<rgba 1 0 0.75 1>}
  n35 = TextureSampleParameter2D(UVs=@n19) {
   group=StormNoise
   parameterName=NoiseTexture
   samplerSource=SSM_Wrap_WorldGroupSettings
   samplerType=SAMPLERTYPE_LinearGrayscale
   texture=/Game/Characters/Player/Male/Medium/Bodies/M_MED_Hurricane/Textures/T_MED_Hurricane_StormCloud
  }
  NamedRerouteDeclaration(None=@n35.R) {name=NoiseSample_1, nodeColor=<rgba 0 0.3438 1 1>}
  n36 = NamedRerouteUsage()
  n37 = VectorParameter() {group=StormNoise, parameterName=NoiseTiling, channelNames {a="", b="", g=Tile Y, r=Tile X}, defaultValue=<rgba 4 4 4 1>}
  n38 = ComponentMask(None=@n37) {g=true, r=true}
  n39 = ScalarParameter() {defaultValue=0.1, group=StormNoise, parameterName=DistortionStrength}
  n40 = NamedRerouteUsage()
  n41 = VectorParameter() {group=StormNoise, parameterName=NoiseSpeed, channelNames {a="", b="", g=Speed Y, r=Speed X}, defaultValue=<rgba 0.025 -0.05 1 1>}
  n42 = Add(A=Add(A=Multiply(A=@n36, B=@n38), B=Multiply(A=@n39, B=@n40)), B=Multiply(A=Time(), B=ComponentMask(None=@n41) {g=true, r=true}))
  n43 = TextureSampleParameter2D(UVs=@n42) {
   group=StormNoise
   parameterName=NoiseTexture
   samplerSource=SSM_Wrap_WorldGroupSettings
   samplerType=SAMPLERTYPE_LinearGrayscale
   texture=&T_MED_Hurricane_StormCloud
  }
  NamedRerouteDeclaration(None=@n43.R) {name=NoiseSample, nodeColor=<rgba 0 0.3437 1 1>}
  NamedRerouteDeclaration(None=@n20.RGB) {name=NoiseNormalSample_1, nodeColor=<rgba 1 0 0.3516 1>}
  n44 = DotProduct(A=CameraVectorWS(), B=Transform(None=NamedRerouteUsage()))
  n45 = Constant()
  n46 = LinearInterpolate(
   A=ScalarParameter() {defaultValue=1.5, group=StormNoise, parameterName=FresnelNoiseMax}
   B=ScalarParameter() {defaultValue=0.5, group=StormNoise, parameterName=FresnelNoiseMin}
   Alpha=NamedRerouteUsage()
  )
  n47 = ScalarParameter() {defaultValue=0.6, group=StormNoise, parameterName=FresnelHardness}
  n48 = NamedRerouteUsage()
  n49 = LinearInterpolate(
   A=ScalarParameter() {defaultValue=0.5, group=StormNoise, parameterName=NoiseSubtractMax}
   B=ScalarParameter() {defaultValue=0.25, group=StormNoise, parameterName=NoiseSubtractMin}
   Alpha=NamedRerouteUsage()
  )
  NamedRerouteDeclaration(None=Subtract(A=SphereMask(A=@n44, B=@n45, Radius=@n46, Hardness=@n47), B=Multiply(A=@n48, B=@n49))) {name=FresnelMask_1, nodeColor=<rgba 1 0.7031 0 1>}
  n50 = NamedRerouteUsage()
  n51 = OneMinus(None=DotProduct(A=CameraVectorWS(), B=Transform(None=@n50)))
  NamedRerouteDeclaration(None=Power(Base=ConstantBiasScale(None=@n51))) {name=NoiseNormalFresnel_1, nodeColor=<rgba 0 1 0.0859 1>}
  n52 = Saturate(None=NamedRerouteUsage())
  n53 = LinearInterpolate(
   A=VectorParameter() {group=StormNoise, parameterName=NoiseColorDark, defaultValue=<rgba 0.0999 0.0268 0.0051 1>}.RGB
   B=VectorParameter() {group=StormNoise, parameterName=NoiseColorLight, defaultValue=<rgba 0.7011 0.1725 0.0472 1>}.RGB
   Alpha=Multiply(A=@n52, B=@n52) {constB=4}
  )
  n54 = NamedRerouteUsage()
  NamedRerouteDeclaration(None=Multiply(A=Multiply(A=@n53, B=@n54), B=ScalarParameter() {defaultValue=2, group=StormNoise, parameterName=GlobalBrightness})) {name=Output_1, nodeColor=<rgba 0.0312 0 1 1>}
  n55 = TextureSampleParameter2D(UVs=@n42) {
   group=StormNoise
   parameterName=NoiseNormal
   samplerSource=SSM_Wrap_WorldGroupSettings
   samplerType=SAMPLERTYPE_Normal
   texture=&T_MED_Hurricane_StormCloud_N
  }
  NamedRerouteDeclaration(None=@n55.RGB) {name=NoiseNormalSample, nodeColor=<rgba 1 0 0.3516 1>}
  n56 = DotProduct(A=CameraVectorWS(), B=Transform(None=NamedRerouteUsage()))
  n57 = Constant()
  n58 = LinearInterpolate(
   A=ScalarParameter() {defaultValue=1, group=StormNoise, parameterName=FresnelNoiseMax_OG}
   B=ScalarParameter() {defaultValue=0.5, group=StormNoise, parameterName=FresnelNoiseMin}
   Alpha=NamedRerouteUsage()
  )
  n59 = ScalarParameter() {group=StormNoise, parameterName=FresnelHardness_0G}
  n60 = NamedRerouteUsage()
  n61 = LinearInterpolate(
   A=ScalarParameter() {defaultValue=0.5, group=StormNoise, parameterName=NoiseSubtractMax}
   B=ScalarParameter() {defaultValue=0.25, group=StormNoise, parameterName=NoiseSubtractMin}
   Alpha=NamedRerouteUsage()
  )
  NamedRerouteDeclaration(None=Subtract(A=SphereMask(A=@n56, B=@n57, Radius=@n58, Hardness=@n59), B=Multiply(A=@n60, B=@n61))) {name=FresnelMask, nodeColor=<rgba 1 0.7031 0 1>}
  n62 = NamedRerouteUsage()
  n63 = OneMinus(None=DotProduct(A=CameraVectorWS(), B=Transform(None=@n62)))
  NamedRerouteDeclaration(None=Power(Base=ConstantBiasScale(None=@n63))) {name=NoiseNormalFresnel, nodeColor=<rgba 0 1 0.0859 1>}
  n64 = Saturate(None=NamedRerouteUsage())
  n65 = LinearInterpolate(
   A=VectorParameter() {group=StormNoise, parameterName=NoiseColorDark, defaultValue=<rgba 0.0999 0.0268 0.0051 1>}.RGB
   B=VectorParameter() {group=StormNoise, parameterName=NoiseColorLight, defaultValue=<rgba 0.7011 0.1725 0.0472 1>}.RGB
   Alpha=Multiply(A=@n64, B=@n64) {constB=4}
  )
  n66 = NamedRerouteUsage()
  NamedRerouteDeclaration(None=Multiply(A=Multiply(A=@n65, B=@n66), B=ScalarParameter() {defaultValue=10, group=StormNoise, parameterName=GlobalBrightness_OG})) {name=Output, nodeColor=<rgba 0.0312 0 1 1>}
  n67 = Fresnel(Normal=@n29) {baseReflectFraction=0, exponent=1}
  n68 = ScalarParameter() {defaultValue=4, parameterName=BumpOffsetPower}
  n69 = Multiply(A=Power(Base=OneMinus(None=@n67), Exp=@n68) {constExponent=20}, B=ScalarParameter() {defaultValue=-6, parameterName=BumpOffset})
  NamedRerouteDeclaration(None=NamedRerouteDeclaration(None=BumpOffset(Coordinate=@n9, Height=@n69)) {name=PolarCoordinates, nodeColor=<rgba 1 0 0.3516 1>}) {name=SphereCoord, nodeColor=<rgba 1 0 0.375 1>}
  n70 = FunctionInput() {bUsePreviewValueAsDefault=true, inputName=Alpha, inputType=FunctionInput_Scalar, sortPriority=1, previewValue=<1 0 0 1>}
  NamedRerouteDeclaration(None=@n70) {name=Alpha, nodeColor=<rgba 0.1484 0 1 1>}
  NamedRerouteUsage()
  OneMinus(None=Step(X=Fresnel() {baseReflectFraction=0, exponent=1}) {constY=0.5})
  Power(Base=OneMinus(None=Fresnel() {baseReflectFraction=0, exponent=1}), Exp=@n68) {constExponent=20}
  VertexNormalWS()
 ]}
