Material /BRCosmetics/Characters/Player/Male/Medium/Bodies/M_MED_SpacePlunge/Materials/M_Glass_SpacePlunge
graph {
 properties {
  bEnableResponsiveAA=true
  bScreenSpaceReflections=true
  bUseMaterialAttributes=true
  bUsedWithMeshDeformer=true
  bUsedWithMorphTargets=true
  bUsedWithSkeletalMesh=true
  blendMode=BLEND_Translucent
  ditherOpacityMask=true
  pixelDepthOffsetMode=PDOM_Legacy
  translucencyDirectionalLightingIntensity=5
  translucencyLightingMode=TLM_SurfacePerPixelLighting
  usedShadingModels=Default Lit
 }
 nodes [
  n1 = SceneTexture() {sceneTextureId=PPI_SceneDepth}
  n2 = Constant()
  ComponentMask(None=ShadingPathSwitch(Default=@n1.Color, Deferred=@n1.Color, Forward=@n2, Mobile=@n2)) {b=true, g=true, r=true}
  Divide(A=Subtract(A=SceneDepth().R, B=PixelDepth().R), B=ScalarParameter() {defaultValue=78.7933, parameterName=FogDepth})
  n3 = TransformPosition(None=WorldPosition().XYZ) {transformSourceType=TRANSFORMPOSSOURCE_World, transformType=TRANSFORMPOSSOURCE_View}
  Length(None=@n3) {is=class MaterialExpressionLength}
  n4 = NamedRerouteUsage()
  LinearInterpolate(
   A=ScalarParameter() {defaultValue=1, parameterName=SmokeMin}
   B=ScalarParameter() {defaultValue=1, parameterName=SmokeHigh}
   Alpha=Power(Base=@n4) {constExponent=5}
  ) {constA=0.1}
  n5 = TextureSampleParameter2D() {
   group=00_Glass
   parameterName=TechArtMask
   samplerType=SAMPLERTYPE_Masks
   sortPriority=-96
   texture=&T_BlackMask
  }
  n6 = MaterialFunctionCall!(
   Material Attributes=MaterialFunctionCall!(TechArtMask=@n5.RGB) {materialFunction=/Game/Global/Materials/Character_Parent_2023/Glass_Parent_2023/MF_Glass}.Result
   Fresnel=Fresnel() {baseReflectFraction=0, exponent=1}
   Alpha=@n5.RGB
  ) {materialFunction=/Game/Global/MaterialFunctions/TechArt/MF_ThinFilm_TechArt}
  n7 = StaticSwitchParameter(
   True=MaterialFunctionCall!(In=@n6.Result) {materialFunction=/Game/Packages/Fortress_SharedMaterials/Base_Material_Functions/BaseCustomizationFunctions/MF_FN_Vehicle_CameraDither}.Result
   False=@n6.Result
  ) {group=00_Glass, parameterName=UseCameraDither, sortPriority=250}
  n8 = Subtract(A=ObjectPositionWS(), B=WorldPosition().XYZ)
  n9 = ScalarParameter() {group=00_Glass, parameterName=VertexHide, sortPriority=301}
  n10 = StaticSwitchParameter(
   True=SetMaterialAttributes!(MaterialAttributes=@n7, World Position Offset=Multiply(A=@n8, B=@n9)) {attributeSetTypes [#1]}
   False=MaterialFunctionCall!(Material=@n7) {materialFunction=/Game/Global/Materials/Character_Parent_2023/Functions/BaseFN_CharacterOpacity_2023}.Result
  ) {group=00_Glass, parameterName=UseVertexHide, sortPriority=300}
  n11 = GetMaterialAttributes(None=@n10) {attributeGetTypes [#2, #3, #4]}
  n12 = NamedRerouteUsage()
  n13 = Multiply(A=@n11.Base Color, B=ScalarParameter() {defaultValue=2, parameterName=BaseColorOverride2})
  n14 = ScalarParameter() {defaultValue=0.4, group=Glass_Extra, parameterName=BaseColorOverride}
  n15 = MaterialFunctionCall!(Material=@n10, Alpha=Constant() {r=1}) {materialFunction=&/MF_StormNoise_Smoke}
  n16 = ComponentMask(None=@n15.FresnelMask) {r=true}
  n17 = Saturate(None=@n16)
  n18 = Add(
   A=LinearInterpolate(
    A=ScalarParameter() {parameterName=SmokeLowOpacity}
    B=ScalarParameter() {defaultValue=0.4, group=Glass_Extra, parameterName=SmokeHighOpacity}
    Alpha=@n17
   ) {constA=0.1, constB=0.5}
   B=NamedRerouteUsage()
  )
  n19 = SetMaterialAttributes!(
   MaterialAttributes=@n11.MaterialAttributes
   Emissive Color=Add(A=NamedRerouteUsage(), B=Add(A=@n11.Emissive Color, B=@n12))
   Base Color=LinearInterpolate(A=Saturate(None=@n13), B=Multiply(A=@n11.Base Color, B=@n14), Alpha=NamedRerouteUsage())
   Opacity=LinearInterpolate(A=Constant() {r=1}, B=Saturate(None=@n18), Alpha=ScalarParameter() {defaultValue=1, group=Glass_Extra_Control, parameterName=IsBoomActive})
   Roughness=Constant() {r=0.2}
  ) {attributeSetTypes [
    #2
    #3
    #4
    #5
   ]}
  MaterialFunctionCall(In=@n19) {materialFunction=&MF_TechArtOverrides}
  Multiply(A=@n16, B=ScalarParameter() {defaultValue=2, group=Glass_Extra, parameterName=SmokeDensity})
  n20 = VectorParameter() {group=Glass_Extra, parameterName=FogColor, defaultValue=<rgba 0.1823 0.0574 0 1>}
  n21 = Multiply(A=@n20.RGB, B=@n20.RGB)
  n22 = SmoothStep(
   Min=ScalarParameter() {group=Fog, parameterName=FogMinStep}
   Max=ScalarParameter() {defaultValue=0.2, group=Fog, parameterName=FogMaxStep}
   Value=NamedRerouteUsage()
  )
  n23 = Power(Base=@n22) {constExponent=0.5}
  n24 = Power(Base=@n15.Result)
  n25 = Multiply(A=Fresnel() {baseReflectFraction=0}, B=ScalarParameter() {defaultValue=1, parameterName=SmokeClose}) {constB=20}
  n26 = ScalarParameter() {group=Fog, parameterName=FogMin}
  n27 = Add(A=Add(A=Multiply(A=@n15.Result, B=@n23), B=Multiply(A=@n24, B=@n25)), B=Multiply(A=LinearInterpolate(A=@n26, Alpha=@n22), B=@n21))
  n28 = VectorParameter() {group=Glass_Extra, parameterName=FresnelColor, defaultValue=<rgba 1 0.3151 0 2>}
  n29 = Multiply(
   A=Power(Base=Fresnel() {baseReflectFraction=0, exponent=1}, Exp=ScalarParameter() {defaultValue=3, parameterName=FresnelPOW})
   B=Multiply(A=@n28.RGB, B=@n28.RGB)
  )
  n30 = Constant()
  n31 = NamedRerouteUsage()
  n32 = ScalarParameter() {defaultValue=1, group=Glass_Extra_Control, parameterName=IsBoomActive}
  NamedRerouteDeclaration(None=LinearInterpolate(A=@n21, B=Add(A=@n27, B=@n29), Alpha=LinearInterpolate(A=@n30, B=@n31, Alpha=@n32))) {name=EM_OUT, nodeColor=<rgba 0 1 0.4609 1>}
  n33 = NamedRerouteUsage()
  n34 = ScalarParameter() {defaultValue=12, group=Blur, parameterName=DistFade}
  n35 = Power(Base=Saturate(None=Divide(A=@n33, B=@n34) {constB=20}), Exp=ScalarParameter() {defaultValue=0.5, group=Blur, parameterName=ScenDepthPower})
  n36 = ScalarParameter() {defaultValue=1, group=Blur, parameterName=ScenDepthIntensity}
  n37 = NamedRerouteDeclaration(None=Saturate(None=Multiply(A=@n35, B=@n36))) {name=SceneDepthAlpha, nodeColor=<rgba 0 0.9297 1 1>}
  NamedRerouteDeclaration(None=LinearInterpolate(
    A=ScalarParameter() {defaultValue=0.2, group=Fog, parameterName=Fog Min Density}
    B=ScalarParameter() {defaultValue=0.9, group=Fog, parameterName=Fog Max Density}
    Alpha=Saturate(None=@n37)
   ) {constA=0.5}) {name=FogMask, nodeColor=<rgba 1 0 0.6328 1>}
  n38 = SceneTexture() {sceneTextureId=PPI_SceneDepth}
  n39 = SceneDepth()
  n40 = ShadingPathSwitch(Default=@n38.Color, Deferred=@n38.Color, Forward=@n39.R, Mobile=@n39.R)
  NamedRerouteDeclaration(None=Subtract(A=ComponentMask(None=@n40) {r=true}, B=NamedRerouteUsage())) {name=CameraFade, nodeColor=<rgba 0.9453 1 0 1>}
  n41 = ComponentMask(None=@n3) {b=true}
  n42 = Subtract(A=@n41, B=ScalarParameter() {group=PiteousKicksPineChest, parameterName=BlurDistanceBias})
  n43 = ScalarParameter() {defaultValue=600, group=PiteousKicksPineChest, parameterName=BlurDistanceFade}
  n44 = LinearInterpolate(
   A=QualitySwitch(
    Default=ScalarParameter() {defaultValue=0.005, group=PiteousKicksPineChest, parameterName=BlurNearDefault}
    Epic=ScalarParameter() {defaultValue=0.005, group=PiteousKicksPineChest, parameterName=BlurNearEpic}
   )
   B=ScalarParameter() {defaultValue=0.0005, group=PiteousKicksPineChest, parameterName=BlurFar}
   Alpha=Saturate(None=Divide(A=@n42, B=@n43))
  )
  n45 = Multiply(A=@n44, B=ScalarParameter() {defaultValue=2, group=PiteousKicksPineChest, parameterName=BlurDitherMulitplier}) {constB=2}
  n46 = MaterialFunctionCall() {materialFunction=&DitherTemporalAA}
  NamedRerouteDeclaration(None=LinearInterpolate(
    A=@n44
    B=LinearInterpolate(A=@n44, B=@n45, Alpha=@n46.Result)
    Alpha=ScalarParameter() {defaultValue=1, group=PiteousKicksPineChest, parameterName=BlurUseDither}
   )) {name=BlurDistance, nodeColor=<rgba 0 1 0.9766 1>}
  n47 = Power(Base=OneMinus(None=@n4))
  n48 = MaterialFunctionCall!(
   Plane Normal Axis=Transform(None=Constant3Vector() {constant=<rgba 0 0 1 0>}) {transformSourceType=TRANSFORMSOURCE_View}
   Plane X Axis=Transform(None=Constant3Vector() {constant=<rgba 1 0 0 0>}) {transformSourceType=TRANSFORMSOURCE_View}
   Plane Center=VectorParameter() {group=StormNoise, parameterName=CubeCenter, defaultValue=<rgba 0 0 0 1>}.RGB
   UV Scale=ScalarParameter() {defaultValue=10, group=Glass_Extra, parameterName=CubeUVScale}
  ) {materialFunction=&/MF_VirtualPlaneCoordinates_SpacePlunge}
  n49 = ScalarParameter() {parameterName=CubeHideCurve}
  n50 = Divide(
   A=Distance(A=NamedRerouteDeclaration(None=@n48.UVs) {name=SphereCoord, nodeColor=<rgba 0.1484 1 0 1>}, B=Constant() {r=0.5})
   B=LinearInterpolate(A=ScalarParameter() {defaultValue=0.6, group=Glass_Extra, parameterName=CubeDistanceCoord}, Alpha=Saturate(None=@n49)) {constB=0}
  ) {constB=0.3}
  n51 = Power(Base=OneMinus(None=Saturate(None=@n50)))
  n52 = NamedRerouteUsage()
  n53 = Multiply(
   A=Multiply(A=Multiply(A=@n51, B=@n52), B=ScalarParameter() {defaultValue=5, group=Glass_Extra, parameterName=BloomIntensity}) {constB=100}
   B=VectorParameter() {group=Glass_Extra, parameterName=BloomColor, defaultValue=<rgba 1 0.7279 0 1>}.RGB
  )
  NamedRerouteDeclaration(None=LinearInterpolate(A=Constant(), B=Multiply(A=@n47, B=@n53), Alpha=ScalarParameter() {defaultValue=1, group=Glass_Extra_Control, parameterName=IsBoomActive})) {name=Bloom, nodeColor=<rgba 1 0 0.3047 1>}
  NamedRerouteDeclaration(None=@n15.SphereMask) {name=NewArea, nodeColor=<rgba 0.7578 1 0 1>}
  NamedRerouteDeclaration(None=@n41) {name=CameraDist, nodeColor=<rgba 0 1 0.5781 1>}
  NamedRerouteDeclaration(None=@n17) {name=SmokeMask, nodeColor=<rgba 1 0 0.4219 1>}
  NamedRerouteUsage()
 ]
}
