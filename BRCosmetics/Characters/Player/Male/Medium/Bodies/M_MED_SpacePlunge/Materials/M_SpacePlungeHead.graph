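# Masked two-sided head material for the SpacePlunge character, adding a reactive scanline effect between the shared character PreFX and PostFX functions.
# MF_Scanlines drives tinted emissive glow and a noise-based opacity cutout restricted to front faces.
# HuntReactivity parameters define a timed burn-up and fade envelope, with a front-end override.
# A view-ray offset behind the surface builds a camera-dependent area mask from dot products.
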Material /BRCosmetics/Characters/Player/Male/Medium/Bodies/M_MED_SpacePlunge/Materials/M_SpacePlungeHead
graph {
 properties {
  bUseMaterialAttributes=true
  bUsedWithMeshDeformer=true
  bUsedWithMorphTargets=true
  bUsedWithSkeletalMesh=true
  blendMode=BLEND_Masked
  pixelDepthOffsetMode=PDOM_Legacy
  shadingModel=MSM_FromMaterialExpression
  twoSided=true
  usedShadingModels=Default Lit | Subsurface | Clear Coat | Subsurface Profile | Cloth
 }
 nodes [
  n1 = MaterialFunctionCall() {materialFunction=&MF_CharacterParent_2023_PreFX}
  n2 = GetMaterialAttributes(None=@n1.Result) {attributeGetTypes [#1]}
  n3 = VectorParameter() {group=Cube_Additional, parameterName=ScanLineColor, defaultValue=<rgba 0.224 0.0371 0 1>}
  n4 = MaterialFunctionCall!(
   PreskinnedPOS=@n1.Pre skinned Local Position
   Scanline Amount=ScalarParameter() {defaultValue=50, group=Stablized, parameterName=ScanlineAmount}
  ) {materialFunction=/BRCosmetics/Accessories/FORT_Backpacks/Backpack_M_MED_SpacePlunge/Material/MF_Scanlines}
  n5 = Saturate(None=LinearInterpolate(
    A=NamedRerouteUsage()
    B=ScalarParameter() {defaultValue=1, group=HuntReactivity, parameterName=ReactiveAlpha}
    Alpha=ScalarParameter() {group=HuntReactivity, parameterName=IsInFrontEnd}
   ))
  n6 = Multiply(A=Multiply(A=Multiply(A=@n3.RGB, B=@n3.RGB), B=@n4.Result), B=@n5)
  n7 = LinearInterpolate(B=LinearInterpolate(A=@n4.SpaceNoise, Alpha=NamedRerouteUsage()), Alpha=@n5) {constA=1}
  n8 = Saturate(None=Sign(None=TwoSidedSign()))
  MaterialFunctionCall(
   Material=SetMaterialAttributes!(
    MaterialAttributes=@n2.MaterialAttributes
    Emissive Color=Add(A=@n2.Emissive Color, B=@n6)
    Opacity Mask=LinearInterpolate(B=@n7, Alpha=@n8) {constA=1}
   ) {attributeSetTypes [#1, #2]}
   M=@n1.M
   Pre Skinned Local Position=@n1.M
   Distance Blend=@n1.M
   Fresnel=@n1.M
  ) {materialFunction=&MF_CharacterParent_2023_PostFX}
  n9 = CameraPositionWS()
  n10 = WorldPosition()
  n11 = ScalarParameter() {defaultValue=-30, parameterName=D}
  n12 = PixelNormalWS()
  n13 = CameraVectorWS()
  n14 = PixelNormalWS()
  n15 = Add(A=@n10.XYZ, B=Divide(A=Multiply(A=@n11, B=@n12), B=DotProduct(A=@n13, B=@n14)))
  n16 = NamedRerouteDeclaration(None=Normalize(VectorInput=Subtract(A=@n9, B=@n15))) {name=NewVector, nodeColor=<rgba 0 0.2969 1 1>}
  n17 = Multiply(A=@n12, B=ScalarParameter() {defaultValue=-50, parameterName=CenterDist})
  n18 = Normalize(VectorInput=Subtract(A=@n15, B=Add(A=@n10.XYZ, B=@n17)))
  NamedRerouteDeclaration(None=Saturate(None=DotProduct(A=@n16, B=@n18))) {name=NewArea, nodeColor=<rgba 0 0.5547 1 1>}
  n19 = Time()
  n20 = ScalarParameter() {group=HuntReactivity, parameterName=ReactiveTime}
  n21 = ScalarParameter() {defaultValue=8, group=HuntReactivity, parameterName=ReactivePhase}
  n22 = Saturate(None=Divide(A=Subtract(A=@n19, B=@n20), B=@n21))
  n23 = Multiply(A=@n22, B=ScalarParameter() {defaultValue=16, parameterName=BurnUpSpeed})
  n24 = ScalarParameter() {defaultValue=2, group=HuntReactivity, parameterName=EndPhase}
  n25 = Divide(A=Subtract(A=@n21, B=@n24), B=@n21)
  n26 = Multiply(A=Subtract(A=@n22, B=@n25), B=Divide(A=@n21, B=@n24))
  NamedRerouteDeclaration(None=LinearInterpolate(A=Saturate(None=@n23), B=OneMinus(None=@n26), Alpha=Step(Y=@n25, X=@n22))) {name=ReactiveAlpha, nodeColor=<rgba 1 0 0.0469 1>}
 ]
}
