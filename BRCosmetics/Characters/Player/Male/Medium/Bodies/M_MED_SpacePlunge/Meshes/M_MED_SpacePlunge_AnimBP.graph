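# Animation Blueprint for the SpacePlunge male medium character body, posing it from its attached parent mesh.
# The copied pose gets an additive offset on the attach_backpack bone, then runs through the SpacePlunge Control Rig, which receives the targeting state.
# When customization loads, it name-matches the backpack mesh against the SpacePlunge pack and stores a not-wearing flag, which the AnimGraph does not read.
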AnimBlueprint /BRCosmetics/Characters/Player/Male/Medium/Bodies/M_MED_SpacePlunge/Meshes/M_MED_SpacePlunge_AnimBP
target_skeleton=&/M_MED_SpacePlunge_Skeleton
graphs [
 AnimGraph {nodes [
   n1 = "Transform (Modify) Bone - Bone: attach_backpack"!(
    ComponentPose=Local To Component(LocalPose=Copy Pose From Mesh() {bUseAttachedParent=true})
    bAlphaBoolEnabled:bool=True
    Translation="0,-3.8,0"
    Rotation="0,0,0"
   ) -> (Pose) {alphaInputType=Bool, rotationMode=BMM_Additive, translationMode=BMM_Additive, boneToModify {boneName=attach_backpack}, translation {y=-3.8}}
   n2 = Control Rig!(Source=Component To Local(ComponentPose=@n1), Alpha, IsTargetingActive:bool=Get bIsTargeting()) -> (Pose) {controlRigAssetReference {blueprintRigClass=&/M_MED_SpacePlunge_CtrlRig.M_MED_SpacePlunge_CtrlRig_C}}
   Output Pose(Result=@n2) {layerGroup=None}
  ]}
 EventGraph {nodes [
   n1 = Event OnCharacterCustomizationLoaded()
   > n2 = Set Backpack mesh(Backpack mesh:Skeletal Mesh Component Object Reference=Get Array Item with EFort Custom Part Type(Array:Array of Skeletal Mesh Component Object References=@n1.CustomMeshParts, PartType=Backpack) -> (Output:Skeletal Mesh Component Object Reference)) -> (Output_Get:Skeletal Mesh Component Object Reference)
   > n3 = Branch(Condition=MatchesWildcard(SourceString=GetDisplayName(Object=@n2), Wildcard=*M_MED_SpacePlunge_Pack))
   > Set Is Not Wearing spaceplung BP(Is Not Wearing spaceplung BP:bool=false) -> (Output_Get:bool)
   Set Is Not Wearing spaceplung BP(execute=@n3.else, Is Not Wearing spaceplung BP=true)
   Event BlueprintUpdateAnimation()
  ]}
]
variables [
 Backpack mesh {type=SkeletalMeshComponent}
 Is Not Wearing spaceplung BP {type=boolean}
]
