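# Animation Blueprint for the face accessory of the male medium SpacePlunge cosmetic body, which has a cube rig.
# The AnimGraph copies the attached parent's pose, offsets base_cube, then applies translation and rotation modifiers to rot_cube.
# Each update computes time-driven ambient spin, sine-based bobbing and a standing-still flag from movement, jump and fall state.
# After evaluation, it writes the Cube_FX socket location to the CubeCenter material parameter.
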AnimBlueprint /BRCosmetics/Characters/Player/Male/Medium/Bodies/M_MED_SpacePlunge/Meshes/Parts/M_MED_SpacePlunge_FaceAcc_AnimBP
target_skeleton=&/M_MED_SpacePlunge_FaceAcc_Skeleton
graphs [
 AnimGraph {nodes [
   n1 = Copy Pose From Mesh() {bCopyCurves=true, bUseAttachedParent=true}
   n2 = "Transform (Modify) Bone - Bone: rot_cube"!(
    ComponentPose="Transform (Modify) Bone - Bone: base_cube"(ComponentPose=Local To Component(LocalPose=@n1), Alpha=1, Translation="0,-2,0", Rotation="0,0,0", Scale="1,1,1") {translationMode=BMM_Additive, translationSpace=BCS_BoneSpace, boneToModify {boneName=base_cube}, translation {y=-2}}
    bAlphaBoolEnabled:bool=True
    Translation="0,0,0"
   ) -> (Pose) {
    alphaInputType=Bool
    rotationSpace=BCS_WorldSpace
    translationMode=BMM_Additive
    translationSpace=BCS_BoneSpace
    alphaBoolBlend {blendInTime=4, blendOption=Cubic, blendOutTime=4}
    boneToModify {boneName=rot_cube}
   }
   n3 = "Transform (Modify) Bone - Bone: rot_cube"!(ComponentPose=@n2, Alpha:float=1, Rotation:Rotator="0,0,0") -> (Pose) {
    rotationMode=BMM_Replace
    rotationSpace=BCS_BoneSpace
    translationSpace=BCS_WorldSpace
    alphaBoolBlend {blendOption=Linear}
    boneToModify {boneName=rot_cube}
   }
   Output Pose(Result=Component To Local(ComponentPose=@n3)) {layerGroup=None}
  ]}
 EventGraph {nodes [
   Event BlueprintUpdateAnimation()
   > n1 = Sequence!(execute) -> (then_0, then_1, then_2:Exec)
   n2 = Get bIsFalling()
   n3 = Get bIsJumping()
   n4 = Get bIsMoving2D()
   Set isStandingStill(execute=@n1.then_2, isStandingStill:bool=AND Boolean(A=NOT Boolean(A=@n4), B=NOR Boolean!(A:bool=@n2, B:bool=@n3) -> (ReturnValue:bool))) -> (Output_Get:bool)
   n5 = float * float(A=float / float(A=GetRealTimeSeconds(), B=-720) {is=title K2Node_CallFunction}, B=10000)
   Set Ambient Rotation(execute=@n1.then_0, Ambient Rotation:Rotator=MakeRotator(Roll=@n5, Yaw=@n5)) -> (Output_Get:Rotator)
   n6 = GetRealTimeSeconds()
   n7 = float * float(A="Sin (Radians)"(A=float * float(A=1.2, B=@n6)), B=0.15)
   n8 = float * float(A="Sin (Radians)"(A=float * float(A=1, B=@n6)), B=0.3)
   n9 = Remap Float(Input="Sin (Radians)"(A=float * float(A=1.4, B=@n6)), OldRange=(X=-1,Y=1), NewRange=(X=-1,Y=2)) -> (Output=~)
   n10 = float * float(A="Clamp (Float)"(Value=@n9, Min=-1, Max=0), B=0.15)
   Set Ambient Translation(execute=@n1.then_1, Ambient Translation:Vector=vector + vector(A=MakeVector(X=@n7, Y=@n8, Z=@n9), B=MakeVector(Y=@n10))) -> (Output_Get:Vector)
   Event BlueprintPostEvaluateAnimation() {is=title K2Node_Event}
   n11 = GetPartSkeletalMeshComponent()
   > Is Valid(InputObject=@n11)
   > SetVectorParameterValueOnMaterials!(
    execute
    self:Mesh Component Object Reference=@n11
    ParameterName:Name=CubeCenter
    ParameterValue:Vector=GetSocketLocation!(self:Scene Component Object Reference=@n11, InSocketName:Name=Cube_FX) -> (ReturnValue:Vector)
   ) -> (then)
  ]}
]
variables [
 Ambient Rotation {type=Rotator}
 Ambient Translation {type=Vector}
 isStandingStill {type=boolean}
]
